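; Replay UI Blueprint that draws aspect ratio bars and a grid over the view through a post-process component.
; UpdateAspectRatio creates a dynamic instance of the given material and sets its BarSize and Letterboxing parameters.
; UpdateGrid creates a dynamic instance of the given grid material.
; Each function rewrites the PPV weighted blendables at full weight, keeping the other overlay's instance when it is valid so both stack.
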(fn UpdateAspectRatio
  (params (Material :type Material%20Interface%20Object%20Reference) (BarSize :type Float%20%28single-precision%29) (Letterboxing :type Float%20%28single-precision%29))
  (bind _returnvalue (Rendering|Material|CreateDynamicMaterialInstance Material "None" "None" :!class /Script/Engine.KismetMaterialLibrary))
  (bind _output_get (Variables|Default|SetDynamicAspectRatio _returnvalue))
  (Rendering|Material|SetScalarParameterValue _output_get "BarSize" BarSize :!class /Script/Engine.MaterialInstanceDynamic)
  (Rendering|Material|SetScalarParameterValue _output_get "Letterboxing" Letterboxing :!class /Script/Engine.MaterialInstanceDynamic)
  (Utilities|IsValid (Variables|Default|GetDynamicGrid)
    (:Is%20Valid
      (Utilities|Struct|SetmembersinPostProcessSettings (Variables|/Script/Engine.PostProcessComponent|GetSettings (Variables|Default|GetPPV)) (Utilities|Struct|MakeWeightedBlendables (Utilities|Array|MakeArray (Utilities|Struct|MakeWeightedBlendable 1.0 (Variables|Default|GetDynamicAspectRatio)) (Utilities|Struct|MakeWeightedBlendable 1.0 (Variables|Default|GetDynamicGrid))))))
    (:Is%20Not%20Valid
      (Utilities|Struct|SetmembersinPostProcessSettings (Variables|/Script/Engine.PostProcessComponent|GetSettings (Variables|Default|GetPPV)) (Utilities|Struct|MakeWeightedBlendables (Utilities|Array|MakeArray (Utilities|Struct|MakeWeightedBlendable 1.0 (Variables|Default|GetDynamicAspectRatio))))))))

(fn UpdateGrid
  (params (Material :type Material%20Interface%20Object%20Reference))
  (bind _returnvalue (Rendering|Material|CreateDynamicMaterialInstance Material "None" "None" :!class /Script/Engine.KismetMaterialLibrary))
  (bind _output_get (Variables|Default|SetDynamicGrid _returnvalue))
  (Utilities|IsValid (Variables|Default|GetDynamicAspectRatio)
    (:Is%20Valid
      (Utilities|Struct|SetmembersinPostProcessSettings (Variables|/Script/Engine.PostProcessComponent|GetSettings (Variables|Default|GetPPV)) (Utilities|Struct|MakeWeightedBlendables (Utilities|Array|MakeArray (Utilities|Struct|MakeWeightedBlendable 1.0 (Variables|Default|GetDynamicAspectRatio)) (Utilities|Struct|MakeWeightedBlendable 1.0 (Variables|Default|GetDynamicGrid))))))
    (:Is%20Not%20Valid
      (Utilities|Struct|SetmembersinPostProcessSettings (Variables|/Script/Engine.PostProcessComponent|GetSettings (Variables|Default|GetPPV)) (Utilities|Struct|MakeWeightedBlendables (Utilities|Array|MakeArray (Utilities|Struct|MakeWeightedBlendable 1.0 _output_get)))))))

(fn ConstructionScript)
