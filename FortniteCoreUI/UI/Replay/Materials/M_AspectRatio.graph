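# Post-process material in the replay UI that masks opposite screen edges into black aspect-ratio bars.
# Letterboxing blends between viewport U and V to select side or top/bottom bars.
# Subtracting BarSize from that coordinate and its inverse, then applying Ceil and multiplying, yields a branchless hard-edged mask.
# The mask blends black with the scene texture, while a separate lerp uses BarOpacity to fade the mask toward one.
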Material /FortniteCoreUI/UI/Replay/Materials/M_AspectRatio
graph {
 properties {bAutomaticallySetUsageInEditor=false, materialDomain=MD_PostProcess, usedShadingModels=Default Lit}
 nodes [
  n1 = ScreenPosition()
  n2 = LinearInterpolate(
   A=ComponentMask(None=@n1.ViewportUV) {r=true}
   B=ComponentMask(None=@n1.ViewportUV) {g=true}
   Alpha=ScalarParameter() {parameterName=Letterboxing}
  )
  n3 = OneMinus(None=@n2)
  n4 = ScalarParameter() {defaultValue=0.15, parameterName=BarSize}
  n5 = Multiply(A=Ceil(None=Subtract(A=@n3, B=@n4)), B=Ceil(None=Subtract(A=@n2, B=@n4)))
  LinearInterpolate(B=SceneTexture() {sceneTextureId=PPI_PostProcessInput0}.Color, Alpha=@n5)
  LinearInterpolate(B=@n5, Alpha=ScalarParameter() {defaultValue=1, parameterName=BarOpacity}) {constA=1}
 ]
}
