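# Translucent UI material for replay interface backplates, drawing a dark, rounded-corner panel in one of four selectable variants.
# A Backplate Switch scalar parameter drives two Switch nodes, one choosing a named color and the other an alpha strength.
# MF_UI_CoreUI_RoundedMask builds the shape from per-corner radii and a pixel-radius toggle, and its mask is multiplied by the chosen strength.
# Variants A to C share the same color, while D is darker.
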Material /FortniteCoreUI/UI/Replay/Materials/M_Replay_Backplate
graph {
 properties {
  bUsedWithStaticMesh=false
  blendMode=BLEND_Translucent
  materialDomain=MD_UI
  pixelDepthOffsetMode=PDOM_Legacy
  usedShadingModels=Default Lit
 }
 nodes [
  Constant3Vector() {constant=<rgba 1 0.4125 0 1>}
  n1 = VectorParameter() {
   group=Shape
   parameterName=Corner Radius
   sortPriority=1
   channelNames {a=BottomRight, b=TopRight, g=BottomLeft, r=TopLeft}
   defaultValue=<rgba 16 16 16 16>
  }
  MaterialExpressionConstant_1 = Constant() {r=0.7}
  MaterialExpressionConstant_2 = Constant() {r=0.4}
  MaterialExpressionConstant_3 = Constant() {r=0.8}
  MaterialExpressionConstant_4 = Constant() {r=0.8}
  n2 = Switch!(
   SwitchValue=ScalarParameter() {parameterName=Backplate Switch, sliderMax=3}
   Default
   A=@MaterialExpressionConstant_1
   B=@MaterialExpressionConstant_2
   C=@MaterialExpressionConstant_3
   D=@MaterialExpressionConstant_4
  ) {inputs [
    {
     inputName=A
     input {
      expression=&self:MaterialExpressionConstant_1
      inputName=None
      mask=0
      maskA=0
      maskB=0
      maskG=0
      maskR=0
      outputIndex=0
     }
    }
    {
     inputName=B
     input {
      expression=&self:MaterialExpressionConstant_2
      inputName=None
      mask=0
      maskA=0
      maskB=0
      maskG=0
      maskR=0
      outputIndex=0
     }
    }
    {
     inputName=C
     input {
      expression=&self:MaterialExpressionConstant_3
      inputName=None
      mask=0
      maskA=0
      maskB=0
      maskG=0
      maskR=0
      outputIndex=0
     }
    }
    {
     inputName=D
     input {
      expression=&self:MaterialExpressionConstant_4
      inputName=None
      mask=0
      maskA=0
      maskB=0
      maskG=0
      maskR=0
      outputIndex=0
     }
    }
   ]}
  Multiply(
   A=MaterialFunctionCall(
    Corner Radius=AppendVector(A=@n1, B=@n1)
    Radius in Pixels=ScalarParameter() {defaultValue=1, group=Shape, parameterName=Radius in Pixels, sliderMax=1, sortPriority=2}
   ) {materialFunction=&MF_UI_CoreUI_RoundedMask}.Shape Mask
   B=@n2
  )
  NamedRerouteDeclaration(None=Constant3Vector() {constant=<rgba 0.0176 0.0232 0.0331 1>}) {name=Backplate A, nodeColor=<rgba 0 0 0 1>}
  NamedRerouteDeclaration(None=Constant3Vector() {constant=<rgba 0.0176 0.0232 0.0331 1>}) {name=Backplate B, nodeColor=<rgba 0 0 0 1>}
  NamedRerouteDeclaration(None=Constant3Vector() {constant=<rgba 0.0176 0.0232 0.0331 1>}) {name=Backplate C, nodeColor=<rgba 0 0 0 1>}
  NamedRerouteDeclaration(None=Constant3Vector() {constant=<rgba 0.0048 0.004 0.0116 1>}) {name=Backplate D, nodeColor=<rgba 0 0 0 1>}
  MaterialExpressionNamedRerouteUsage_2 = NamedRerouteUsage()
  MaterialExpressionNamedRerouteUsage_6 = NamedRerouteUsage()
  MaterialExpressionNamedRerouteUsage_7 = NamedRerouteUsage()
  MaterialExpressionNamedRerouteUsage_8 = NamedRerouteUsage()
  Switch(
   SwitchValue=ScalarParameter() {parameterName=Backplate Switch, sliderMax=3}
   A=@MaterialExpressionNamedRerouteUsage_2
   B=@MaterialExpressionNamedRerouteUsage_6
   C=@MaterialExpressionNamedRerouteUsage_7
   D=@MaterialExpressionNamedRerouteUsage_8
  ) {inputs [
    {
     inputName=A
     input {
      expression=&self:MaterialExpressionNamedRerouteUsage_2
      inputName=None
      mask=0
      maskA=0
      maskB=0
      maskG=0
      maskR=0
      outputIndex=0
     }
    }
    {
     inputName=B
     input {
      expression=&self:MaterialExpressionNamedRerouteUsage_6
      inputName=None
      mask=0
      maskA=0
      maskB=0
      maskG=0
      maskR=0
      outputIndex=0
     }
    }
    {
     inputName=C
     input {
      expression=&self:MaterialExpressionNamedRerouteUsage_7
      inputName=None
      mask=0
      maskA=0
      maskB=0
      maskG=0
      maskR=0
      outputIndex=0
     }
    }
    {
     inputName=D
     input {
      expression=&self:MaterialExpressionNamedRerouteUsage_8
      inputName=None
      mask=0
      maskA=0
      maskB=0
      maskG=0
      maskR=0
      outputIndex=0
     }
    }
   ]}
 ]
}
