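# TextRotator is a CommonUI rotator widget that displays its current text option centered inside a styled border.
# The inherited MyText block sits in MainBorder, which fills a horizontal box and uses a text scroll style on the label.
# PreConstruct sets the border brush to the Image variable when bChangeAppearance is true, otherwise to the dark-blue ItemInfo material.
# Construct makes the widget focusable.
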WidgetBlueprint /FortUILibrary/Foundation/Buttons/TextRotator
graphs [EventGraph {nodes [
   Event PreConstruct()
   > n1 = Branch(Condition=Get bChangeAppearance() -> (bChangeAppearance:bool))
   > SetBrush!(
    execute
    self:Border Object Reference=Get MainBorder() -> (MainBorder:Common Border Object Reference)
    InBrush_TintColor:Slate Color Structure=(SpecifiedColor=(R=1,G=1,B=1,A=1),ColorUseRule=UseColor_Specified)
    InBrush_DrawAs:ESlateBrushDrawType Enum=Image
    InBrush_Tiling:ESlateBrushTileType Enum=NoTile
    InBrush_Mirroring:ESlateBrushMirrorType Enum=NoMirror
    InBrush_ImageSize_X:double=32
    InBrush_ImageSize_Y:double=32
    InBrush_Margin:Margin Structure=(Left=0,Top=0,Right=0,Bottom=0)
    InBrush_ResourceObject:Object Reference=Get Image() -> (Image:Object Reference)
    InBrush_OutlineSettings:Slate Brush Outline Settings Structure=(CornerRadii=(X=0,Y=0,Z=0,W=0),Color=(SpecifiedColor=(R=0,G=0,B=0,A=0),ColorUseRule=UseColor_Specified),Width=0,RoundingType=HalfHeightRadius,bUseBrushTransparency=False)
   ) -> (then)
   SetBrush(
    execute=@n1.else
    self=Get MainBorder()
    InBrush_TintColor=(SpecifiedColor=(R=1,G=1,B=1,A=1),ColorUseRule=UseColor_Specified)
    InBrush_DrawAs=Image
    InBrush_Tiling=NoTile
    InBrush_Mirroring=NoMirror
    InBrush_ImageSize_X=32
    InBrush_ImageSize_Y=32
    InBrush_Margin=(Left=0,Top=0,Right=0,Bottom=0)
    InBrush_ResourceObject=/FortUILibrary/Foundation/Buttons/Materials/M-ItemInfo-Blank-DkBlue
    InBrush_OutlineSettings=(CornerRadii=(X=0,Y=0,Z=0,W=0),Color=(SpecifiedColor=(R=0,G=0,B=0,A=0),ColorUseRule=UseColor_Specified),Width=0,RoundingType=HalfHeightRadius,bUseBrushTransparency=False)
   )
   Event Construct()
   > Set bIsFocusable(bIsFocusable:bool=true) -> (Output_Get:bool)
  ]}]
variables [
 bChangeAppearance {type=boolean}
 Image {type=Object, default=null}
]
widgets {
 info {inheritedWidgetCount=2, parentClass=/Script/CommonUI.CommonRotator, rootWidgetClass=&HorizontalBox, widgetCount=4}
 widgets [
  HorizontalBox_0 {widgetClassPath=&HorizontalBox, widget=HorizontalBox_0}
  MainBorder {
   widgetClassPath=&CommonBorder
   bIsVariable=true
   parent=HorizontalBox_0
   widget=MainBorder
   properties {
    navigation=&self:WidgetTree.MainBorder.WidgetNavigation_0
    style=/FortUILibrary/Styles/Border/Border-ItemInfo-Blank-DkBlue.Border-ItemInfo-Blank-DkBlue_C
    verticalAlignment=VAlign_Center
    visibility=Visible
    background {
     bIsDynamicallyLoaded=false
     drawAs=Box
     imageType=NoImage
     mirroring=NoMirror
     resourceName=None
     resourceObject=/FortUILibrary/Foundation/Buttons/Materials/M-ItemInfo-Blank
     tiling=NoTile
     imageSize=<32 32>
     margin=<ltrb 0.4 0.4 0.4 0.4>
     outlineSettings {
      bUseBrushTransparency=false
      roundingType=HalfHeightRadius
      width=0
      color {colorUseRule=UseColor_Specified, specifiedColor=<rgba 0 0 0 0>}
      cornerRadii=<0 0 0 0>
     }
     tintColor {colorUseRule=UseColor_Specified, specifiedColor=<rgba 1 1 1 1>}
     uVRegion {bIsValid=false, max=<0 0>, min=<0 0>}
    }
    padding=<ltrb 0 0 0 0>
   }
   slot {class=HorizontalBoxSlot, properties {size {sizeRule=Fill, value=1}}}
  }
  MyText {
   widgetClassPath=&CommonTextBlock
   bInherited=true
   bIsVariable=true
   parent=MainBorder
   widget=MyText
   properties {
    justification=Center
    scrollStyle=&TextScrollStyle-Base_C
    text=Text Block
    font {
     bForceMonospaced=false
     bMaterialIsStencil=false
     fontMaterial=null
     fontObject=&NotoSans
     letterSpacing=0
     monospacedWidth=1
     size=16
     skewAmount=0
     typefaceFontName=NotoSans-Bold
     outlineSettings {
      bApplyOutlineToDropShadows=false
      bMiteredCorners=false
      bSeparateFillAlpha=false
      outlineMaterial=null
      outlineSize=0
      outlineColor=<rgba 0 0 0 1>
     }
    }
    shadowOffset=<0 0>
   }
   slot {class=BorderSlot, properties {verticalAlignment=VAlign_Center, padding=<ltrb 0 0 0 0>}}
  }
  InputActionWidget {widgetClassPath=&CommonActionWidget, bInherited=true, widget=null}
 ]
}
