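# Replay settings row widget, based on ReplayColorPickerComponent, that serves as either a hue picker or a numeric slider, chosen by bColor.
# PreConstruct shows a hue bar with chevron handle in color mode, otherwise value text.
# Slider changes move the overlaid arrow slider, update progress and text in numeric mode, and broadcast ColorChanged only when LightingTool is set.
# Mouse hover animates the highlight, plays a sound, broadcasts OnHovered and shows the setting description.
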WidgetBlueprint /FortniteCoreUI/UI/Replay/ColorPickerReplay
animations [OnHover {
  binding_count=3
  length_seconds=0.1
  object=&self:OnHover
  sequence {bindings [
    FocusOutline {
     id=#1
     possessed_class=Image
     tracks [{
       type=MovieSceneFloatTrack
       display_name=Render Opacity
       sections [{type=MovieSceneFloatSection, end_frame=2, start_frame=0, channels [None {type=MovieSceneScriptingFloatChannel, keys [[0,0],[2,1]]}]}]
      }]
    }
    Text_ComponentName {
     id=#2
     possessed_class=CommonTextBlock
     tracks [{
       type=MovieSceneColorTrack
       display_name=Color and Opacity
       sections [{
         type=MovieSceneColorSection
         end_frame=2
         start_frame=0
         channels [
          Color.R {type=MovieSceneScriptingFloatChannel, keys [[0,0.521],[2,1]]}
          Color.G {type=MovieSceneScriptingFloatChannel, keys [[0,0.521],[2,1]]}
          Color.B {type=MovieSceneScriptingFloatChannel, keys [[0,0.521],[2,1]]}
          Color.A {type=MovieSceneScriptingFloatChannel, keys [[0,0.8],[2,1]]}
         ]
        }]
      }]
    }
    Text_SettingValue {
     id=#3
     possessed_class=CommonTextBlock
     tracks [{
       type=MovieSceneColorTrack
       display_name=Color and Opacity
       sections [{
         type=MovieSceneColorSection
         end_frame=2
         start_frame=0
         channels [
          Color.R {type=MovieSceneScriptingFloatChannel, keys [[0,0.521],[2,1]]}
          Color.G {type=MovieSceneScriptingFloatChannel, keys [[0,0.521],[2,1]]}
          Color.B {type=MovieSceneScriptingFloatChannel, keys [[0,0.521],[2,1]]}
          Color.A {type=MovieSceneScriptingFloatChannel, keys [[0,0.8],[2,1]]}
         ]
        }]
      }]
    }
   ]}
 }]
graphs [
 ColorChanged {nodes [ColorChanged!() -> (then, Value:double=~) {is=title K2Node_FunctionEntry}]}
 EventGraph {nodes [
   n1 = "On Value Changed (ColorSlider_Component)"!() -> (OutputDelegate, then, Value:float) {is=title K2Node_ComponentBoundEvent}
   > n2 = Set SliderValue(SliderValue:float=@n1.Value) -> (Output_Get:float)
   > SetValue!(
    execute
    self:Slider Object Reference=Get ColorSlider_Arrow() -> (ColorSlider_Arrow:Replay Color Picker Slider Object Reference)
    InValue:float=@n2
   ) -> (then)
   > n3 = Branch(Condition=Get bColor() -> (bColor:bool))
   Branch(execute=[@n4.then, @n3.then], Condition=Get LightingTool() -> (LightingTool:bool))
   > Call Color Changed!(execute:Exec, self:Color Picker Replay Object Reference, Value:double=@n2) -> (then:Exec) {is=title K2Node_CallDelegate}
   SetProgress!(
    execute=@n3.else
    self:Fort Simple Material Progress Bar Object Reference=Get Progress() -> (Progress:Simple Material Progress Bar Object Reference)
    Progress:float=@n2
   ) -> (then)
   > n4 = "SetText (Text)"(
    self=Get Text_SettingValue() -> (Text_SettingValue:Common Text Object Reference)
    InText="To Text (Float)"(Value=@n2, bUseGrouping=false, MinimumIntegralDigits=1, MaximumIntegralDigits=100, MinimumFractionalDigits=2, MaximumFractionalDigits=2)
   )
   Event PreConstruct()
   n5 = Get WBP_UIKit_Backplate!(self:Setting Entry Background Object Reference=Get Background() -> (Background:Setting Entry Background Object Reference)) -> (WBP_UIKit_Backplate:WBP UIKit Backplate Object Reference)
   > SetBackplateMaterial!(
    execute
    self:WBP UIKit Backplate Base Object Reference=@n5
    Material Instance:Material Instance Object Reference=/FortniteCoreUI/UI/Replay/Materials/M_Replay_Backplate_Inst
   ) -> (then)
   > SetBrightness!(execute, self:WBP UIKit Backplate Base Object Reference=@n5, Brightness:E_UI_BackplateBrightness Enum=NewEnumerator1) -> (then)
   > n6 = Branch(Condition=Get bColor())
   > SetBrushFromMaterial(
    self=Get Image_Bar!(self:Fort Simple Material Progress Bar Object Reference=Get Progress()) -> (Image_Bar:Image Object Reference)
    Material=/FortniteCoreUI/UI/Replay/Materials/M_UI_Replay_ColorSlider_Hue
   )
   > SetSliderHandleColor!(execute, self:Slider Object Reference=Get ColorSlider_Arrow(), InValue:Linear Color Structure=(R=1,G=1,B=1,A=1)) -> (then)
   > SetVisibility(self=Get Text_SettingValue(), InVisibility=Collapsed)
   SetSliderHandleColor(execute=@n6.else, self=Get ColorSlider_Arrow(), InValue=(R=1,G=1,B=1,A=0))
   > SetVisibility(self=Get Text_SettingValue(), InVisibility=SelfHitTestInvisible)
   Event OnMouseLeave!() -> (OutputDelegate, then, MouseEvent:"Pointer Event Structure (by ref)") {is=title K2Node_Event}
   > PlayAnimationReverse(InAnimation=Get OnHover() -> (OnHover:Widget Animation Object Reference))
   Event OnMouseEnter()
   > PlaySound2D(Sound=/Game/Sounds/Fort_UI_Sounds/FrontEnd/DefaultSounds/PressedButtons/MSS_UI_Button_Hover_Default)
   > PlayAnimationForward(InAnimation=Get OnHover())
   > Call On Hovered!(execute:Exec, self:Color Picker Replay Object Reference) -> (then:Exec) {is=title K2Node_CallDelegate}
   > ShowSettingDescription!(
    execute
    self:Fort Replay Context Object Reference=Get FortReplayContext!() -> (ReturnValue:Fort Replay Context Object Reference) {is=title K2Node_GetBlueprintContextEx}
    Header:"Text (by ref)"=Get Text(self:Text Object Reference=Get Text_ComponentName() -> (Text_ComponentName:Common Text Object Reference))
    Description:"Text (by ref)"=Get Description() -> (Description:Text)
   ) -> (then)
  ]}
 OnHovered {nodes [OnHovered() {is=title K2Node_FunctionEntry}]}
 UpdateSliderValue {nodes [
   n1 = UpdateSliderValue!() -> (then, Value:double=~) {is=title K2Node_FunctionEntry}
   > n2 = Set SliderValue(SliderValue=@n1)
   > SetValue(self=Get ColorSlider_Component() -> (ColorSlider_Component:Replay Color Picker Slider Object Reference), InValue=@n2)
  ]}
]
variables [
 bColor {type=boolean, default=true}
 ColorChanged {}
 Description {type=string, default=""}
 LightingTool {type=boolean}
 OnHovered {}
 SliderValue {type=number}
]
widgets {
 info {
  inheritedWidgetCount=2
  namedSlotCount=1
  parentClass=/Script/FortniteUI.ReplayColorPickerComponent
  rootWidgetClass=&SizeBox
  widgetCount=14
 }
 widgets [
  SizeBox_0 {widgetClassPath=&SizeBox, widget=SizeBox_0, properties {bOverride_WidthOverride=true, widthOverride=650}}
  Overlay_38 {widgetClassPath=&Overlay, parent=SizeBox_0, widget=Overlay_38, slot {class=SizeBoxSlot}}
  Background {
   widgetClassPath=/FortniteCoreUI/Settings/SettingEntries/SettingEntryBackground
   bIsVariable=true
   parent=Overlay_38
   widget=Background
   slot {class=OverlaySlot, properties {horizontalAlignment=HAlign_Fill, verticalAlignment=VAlign_Fill}}
  }
  Overlay {widgetClassPath=&Overlay, namedSlotHost=Background, widget=Overlay}
  FortResponsivePanel_0 {
   widgetClassPath=/Script/FortniteUI.FortResponsivePanel
   parent=Overlay
   widget=FortResponsivePanel_0
   properties {bCanStackVertically=false, bOverride_Cursor=true}
   slot {class=OverlaySlot, properties {horizontalAlignment=HAlign_Fill}}
  }
  SizeBox2 {
   widgetClassPath=&SizeBox
   parent=FortResponsivePanel_0
   widget=SizeBox2
   properties {bOverride_MinDesiredHeight=true, minDesiredHeight=64}
   slot {class=FortResponsivePanelSlot}
  }
  Text_ComponentName {
   widgetClassPath=&CommonTextBlock
   bInherited=true
   bIsVariable=true
   parent=SizeBox2
   widget=Text_ComponentName
   properties {
    renderOpacity=0.75
    scrollStyle=&TextScrollStyle-Base_C
    text="{Setting Name}"
    colorAndOpacity {colorUseRule=UseColor_Specified, specifiedColor=<rgba 0.521 0.521 0.521 0.8>}
    font {
     bForceMonospaced=false
     bMaterialIsStencil=false
     fontMaterial=null
     fontObject=&HeadingNow
     letterSpacing=0
     monospacedWidth=1
     size=18
     skewAmount=0
     typefaceFontName=75 Medium
     outlineSettings {
      bApplyOutlineToDropShadows=false
      bMiteredCorners=false
      bSeparateFillAlpha=false
      outlineMaterial=null
      outlineSize=0
      outlineColor=<rgba 0 0 0 1>
     }
    }
    shadowOffset=<0 0>
   }
   slot {class=SizeBoxSlot, properties {horizontalAlignment=HAlign_Left, verticalAlignment=VAlign_Center, padding=<ltrb 24 0 10 0>}}
  }
  SizeBox3 {
   widgetClassPath=&SizeBox
   parent=FortResponsivePanel_0
   widget=SizeBox3
   properties {bOverride_MinDesiredHeight=true, minDesiredHeight=50}
   slot {class=FortResponsivePanelSlot}
  }
  PanelValue {widgetClassPath=&Overlay, parent=SizeBox3, widget=PanelValue, slot {class=SizeBoxSlot}}
  Progress {
   widgetClassPath=/Game/Athena/HUD/SimpleMaterialProgressBar
   bIsVariable=true
   parent=PanelValue
   widget=Progress
   properties {
    bCachedEnableColoredSegments=true
    bOverrideDefaultColorA=true
    bOverrideDefaultColorB=true
    bOverrideDefaultColorBackground=true
    segments=1
    visibility=HitTestInvisible
    cachedColorA=<rgba 0 0.0232 0.2502 1>
    cachedColorB=<rgba 0 0.0232 0.2502 1>
    cachedColorBackground=<rgba 0 0 0.0331 1>
    cachedSegmentColor=<rgba 0 0 0 1>
   }
   slot {class=OverlaySlot, properties {horizontalAlignment=HAlign_Fill, verticalAlignment=VAlign_Fill}}
  }
  ColorSlider_Component {
   widgetClassPath=/Script/FortniteUI.ReplayColorPickerSlider
   bInherited=true
   bIsVariable=true
   parent=PanelValue
   widget=ColorSlider_Component
   properties {sliderBarColor=<rgba 1 1 1 0>, sliderHandleColor=<rgba 1 1 1 0>}
   slot {class=OverlaySlot, properties {horizontalAlignment=HAlign_Fill, verticalAlignment=VAlign_Fill}}
  }
  ColorSlider_Arrow {
   widgetClassPath=&ReplayColorPickerSlider
   bIsVariable=true
   parent=PanelValue
   widget=ColorSlider_Arrow
   properties {
    visibility=SelfHitTestInvisible
    renderTransform {angle=0, scale=<1 1>, shear=<0 0>, translation=<0 52>}
    sliderBarColor=<rgba 1 1 1 0>
    sliderHandleColor=<rgba 1 1 1 0>
    widgetStyle {
     barThickness=2
     disabledBarImage {
      bIsDynamicallyLoaded=false
      drawAs=Image
      imageType=NoImage
      mirroring=NoMirror
      resourceName=None
      resourceObject=null
      tiling=NoTile
      imageSize=<0 0>
      margin=<ltrb 0 0 0 0>
      outlineSettings {
       bUseBrushTransparency=false
       roundingType=HalfHeightRadius
       width=0
       color {colorUseRule=UseColor_Specified, specifiedColor=<rgba 0 0 0 0>}
       cornerRadii=<0 0 0 0>
      }
      tintColor {colorUseRule=UseColor_Specified, specifiedColor=<rgba 0.5 0.5 0.5 1>}
      uVRegion {bIsValid=false, max=<0 0>, min=<0 0>}
     }
     disabledThumbImage {
      bIsDynamicallyLoaded=false
      drawAs=Image
      imageType=FullColor
      mirroring=NoMirror
      resourceName=None
      resourceObject=/FortUILibrary/Textures/MaterialInstances/MI_UIKit_Icon_Chevron_Up
      tiling=NoTile
      imageSize=<18 18>
      margin=<ltrb 0 0 0 0>
      outlineSettings {
       bUseBrushTransparency=false
       roundingType=HalfHeightRadius
       width=0
       color {colorUseRule=UseColor_Specified, specifiedColor=<rgba 0 0 0 0>}
       cornerRadii=<0 0 0 0>
      }
      tintColor {colorUseRule=UseColor_Specified, specifiedColor=<rgba 1 1 1 1>}
      uVRegion {bIsValid=false, max=<0 0>, min=<0 0>}
     }
     hoveredBarImage {
      bIsDynamicallyLoaded=false
      drawAs=Image
      imageType=NoImage
      mirroring=NoMirror
      resourceName=None
      resourceObject=null
      tiling=NoTile
      imageSize=<0 0>
      margin=<ltrb 0 0 0 0>
      outlineSettings {
       bUseBrushTransparency=false
       roundingType=HalfHeightRadius
       width=0
       color {colorUseRule=UseColor_Specified, specifiedColor=<rgba 0 0 0 0>}
       cornerRadii=<0 0 0 0>
      }
      tintColor {colorUseRule=UseColor_Specified, specifiedColor=<rgba 1 1 1 1>}
      uVRegion {bIsValid=false, max=<0 0>, min=<0 0>}
     }
     hoveredThumbImage {
      bIsDynamicallyLoaded=false
      drawAs=Image
      imageType=FullColor
      mirroring=NoMirror
      resourceName=None
      resourceObject=&MI_UIKit_Icon_Chevron_Up
      tiling=NoTile
      imageSize=<18 18>
      margin=<ltrb 0 0 0 0>
      outlineSettings {
       bUseBrushTransparency=false
       roundingType=HalfHeightRadius
       width=0
       color {colorUseRule=UseColor_Specified, specifiedColor=<rgba 0 0 0 0>}
       cornerRadii=<0 0 0 0>
      }
      tintColor {colorUseRule=UseColor_Specified, specifiedColor=<rgba 1 1 1 1>}
      uVRegion {bIsValid=false, max=<0 0>, min=<0 0>}
     }
     normalBarImage {
      bIsDynamicallyLoaded=false
      drawAs=Image
      imageType=NoImage
      mirroring=NoMirror
      resourceName=None
      resourceObject=null
      tiling=NoTile
      imageSize=<0 0>
      margin=<ltrb 0 0 0 0>
      outlineSettings {
       bUseBrushTransparency=false
       roundingType=HalfHeightRadius
       width=0
       color {colorUseRule=UseColor_Specified, specifiedColor=<rgba 0 0 0 0>}
       cornerRadii=<0 0 0 0>
      }
      tintColor {colorUseRule=UseColor_Specified, specifiedColor=<rgba 1 1 1 1>}
      uVRegion {bIsValid=false, max=<0 0>, min=<0 0>}
     }
     normalThumbImage {
      bIsDynamicallyLoaded=false
      drawAs=Image
      imageType=FullColor
      mirroring=NoMirror
      resourceName=None
      resourceObject=&MI_UIKit_Icon_Chevron_Up
      tiling=NoTile
      imageSize=<18 18>
      margin=<ltrb 0 0 0 0>
      outlineSettings {
       bUseBrushTransparency=false
       roundingType=HalfHeightRadius
       width=0
       color {colorUseRule=UseColor_Specified, specifiedColor=<rgba 0 0 0 0>}
       cornerRadii=<0 0 0 0>
      }
      tintColor {colorUseRule=UseColor_Specified, specifiedColor=<rgba 1 1 1 1>}
      uVRegion {bIsValid=false, max=<0 0>, min=<0 0>}
     }
    }
   }
   slot {class=OverlaySlot, properties {horizontalAlignment=HAlign_Fill, verticalAlignment=VAlign_Fill, padding=<ltrb -8 0 -8 0>}}
  }
  Text_SettingValue {
   widgetClassPath=&CommonTextBlock
   bIsVariable=true
   parent=PanelValue
   widget=Text_SettingValue
   properties {
    justification=Center
    scrollStyle=&TextScrollStyle-Base_C
    text=1.00
    colorAndOpacity {colorUseRule=UseColor_Specified, specifiedColor=<rgba 0.521 0.521 0.521 0.8>}
    font {
     bForceMonospaced=false
     bMaterialIsStencil=false
     fontMaterial=null
     fontObject=&HeadingNow
     letterSpacing=0
     monospacedWidth=1
     size=18
     skewAmount=0
     typefaceFontName=74 Regular
     outlineSettings {
      bApplyOutlineToDropShadows=false
      bMiteredCorners=false
      bSeparateFillAlpha=false
      outlineMaterial=null
      outlineSize=0
      outlineColor=<rgba 0 0 0 1>
     }
    }
    shadowOffset=<0 0>
   }
   slot {class=OverlaySlot, properties {horizontalAlignment=HAlign_Fill, verticalAlignment=VAlign_Center}}
  }
  FocusOutline {
   widgetClassPath=&Image
   bIsVariable=true
   parent=Overlay
   widget=FocusOutline
   properties {
    pixelSnapping=Disabled
    renderOpacity=0
    visibility=SelfHitTestInvisible
    brush {
     bIsDynamicallyLoaded=false
     drawAs=Image
     imageType=NoImage
     mirroring=NoMirror
     resourceName=None
     resourceObject=/FortniteCoreUI/Settings/Art/MI_UIKit_Settings_Outline
     tiling=NoTile
     imageSize=<32 32>
     margin=<ltrb 0 0 0 0>
     outlineSettings {
      bUseBrushTransparency=false
      roundingType=HalfHeightRadius
      width=0
      color {colorUseRule=UseColor_Specified, specifiedColor=<rgba 0 0 0 0>}
      cornerRadii=<0 0 0 0>
     }
     tintColor {colorUseRule=UseColor_Specified, specifiedColor=<rgba 1 1 1 1>}
     uVRegion {bIsValid=false, max=<0 0>, min=<0 0>}
    }
   }
   slot {class=OverlaySlot, properties {horizontalAlignment=HAlign_Fill, verticalAlignment=VAlign_Fill, padding=<ltrb -16 -16 -16 -16>}}
  }
 ]
}
