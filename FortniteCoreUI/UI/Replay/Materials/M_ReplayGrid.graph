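# Post-process material for the replay UI that draws a rule-of-thirds grid over the scene.
# Distances from screen UVs to the one-third and two-thirds marks on each axis pass through SmoothStep to form thin, soft-edged lines.
# Horizontal-line distances are scaled by the view's height-to-width ratio to keep thickness even.
# The summed, clamped mask blends post-process input scene color toward the LineColor parameter.
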Material /FortniteCoreUI/UI/Replay/Materials/M_ReplayGrid
graph {
 properties {bAutomaticallySetUsageInEditor=false, materialDomain=MD_PostProcess, usedShadingModels=Default Lit}
 nodes [
  n1 = Constant() {r=0.0008}
  n2 = Constant() {r=0.0018}
  n3 = ScreenPosition()
  n4 = ComponentMask(None=@n3.ViewportUV) {desc=Horizontal, r=true}
  n5 = Abs(None=Subtract(A=@n4) {constB=0.333})
  n6 = Abs(None=Subtract(A=@n4) {constB=0.667})
  n7 = Add(
   A=OneMinus(None=SmoothStep(Min=@n1, Max=@n2, Value=@n5) {constMax=0.007, constMin=0.003})
   B=OneMinus(None=SmoothStep(Min=@n1, Max=@n2, Value=@n6) {constMax=0.007, constMin=0.003})
  )
  n8 = ComponentMask(None=@n3.ViewportUV) {desc=Vertical, g=true}
  n9 = ViewSize()
  n10 = Divide(A=ComponentMask(None=@n9) {desc=Height, g=true}, B=ComponentMask(None=@n9) {desc=Width, r=true})
  n11 = Multiply(A=Abs(None=Subtract(A=@n8) {constB=0.333}), B=@n10) {constB=0.56}
  n12 = Multiply(A=Abs(None=Subtract(A=@n8) {constB=0.667}), B=@n10) {constB=0.56}
  n13 = Add(
   A=OneMinus(None=SmoothStep(Min=@n1, Max=@n2, Value=@n11) {constMax=0.007, constMin=0.003})
   B=OneMinus(None=SmoothStep(Min=@n1, Max=@n2, Value=@n12) {constMax=0.007, constMin=0.003})
  )
  LinearInterpolate(
   A=ComponentMask(None=SceneTexture() {sceneTextureId=PPI_PostProcessInput0}.Color) {b=true, g=true, r=true}
   B=VectorParameter() {parameterName=LineColor, defaultValue=<rgba 1 1 1 0>}.RGB
   Alpha=Clamp(None=Add(A=@n7, B=@n13))
  )
 ]
}
