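# Replay options menu row widget showing a setting title beside an action button on a settings-style backplate.
# PreConstruct applies the title, with none, one or two indents by Formatting, plus the button text, and sets the replay backplate material and brightness.
# Mouse enter and leave play the OnHover highlight and fire dispatchers; entering also plays a sound and shows the description through FortReplayContext.
# Button clicks fire ButtonClicked, and receiving focus centers the button.
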WidgetBlueprint /FortniteCoreUI/UI/Replay/OptionsMenuActionReplay
animations [OnHover {
  binding_count=2
  length_seconds=0.1
  object=&self:OnHover
  sequence {bindings [
    Text_SettingName {
     id=#1
     possessed_class=CommonTextBlock
     tracks [{
       type=MovieSceneColorTrack
       display_name=Color and Opacity
       sections [{
         type=MovieSceneColorSection
         end_frame=2
         start_frame=0
         channels [
          Color.R {type=MovieSceneScriptingFloatChannel, keys [[0,0.521],[2,1]]}
          Color.G {type=MovieSceneScriptingFloatChannel, keys [[0,0.521],[2,1]]}
          Color.B {type=MovieSceneScriptingFloatChannel, keys [[0,0.521],[2,1]]}
          Color.A {type=MovieSceneScriptingFloatChannel, keys [[0,0.8],[2,1]]}
         ]
        }]
      }]
    }
    FocusOutline {
     id=#2
     possessed_class=Image
     tracks [{
       type=MovieSceneFloatTrack
       display_name=Render Opacity
       sections [{type=MovieSceneFloatSection, end_frame=2, start_frame=0, channels [None {type=MovieSceneScriptingFloatChannel, keys [[0,0],[2,1]]}]}]
      }]
    }
   ]}
 }]
graphs [
 ButtonClicked {nodes [ButtonClicked() {is=title K2Node_FunctionEntry}]}
 ButtonHovered {nodes [ButtonHovered() {is=title K2Node_FunctionEntry}]}
 ButtonUnhovered {nodes [ButtonUnhovered() {is=title K2Node_FunctionEntry}]}
 EventGraph {nodes [
   "On Clicked (Button_Action)"!() -> (OutputDelegate, then, Button:Common Button Base Object Reference) {is=title K2Node_ComponentBoundEvent}
   > Call Button Clicked!(execute:Exec, self:Options Menu Action Replay Object Reference) -> (then:Exec) {is=title K2Node_CallDelegate}
   Event PreConstruct()
   > Initialize!(
    execute
    self
    TitleText:Text=Get Title() -> (Title:Text)
    ButtonText:Text=Get ButtonText() -> (ButtonText:Text)
    Formatting:E_Replay_TextFormat Enum=Get Formatting() -> (Formatting:E_Replay_TextFormat Enum=NewEnumerator2)
   ) -> (then)
   n1 = Get WBP_UIKit_Backplate!(self:Setting Entry Background Object Reference=Get Background() -> (Background:Setting Entry Background Object Reference)) -> (WBP_UIKit_Backplate:WBP UIKit Backplate Object Reference)
   > SetBackplateMaterial!(
    execute
    self:WBP UIKit Backplate Base Object Reference=@n1
    Material Instance:Material Instance Object Reference=/FortniteCoreUI/UI/Replay/Materials/M_Replay_Backplate_Inst
   ) -> (then)
   > SetBrightness!(execute, self:WBP UIKit Backplate Base Object Reference=@n1, Brightness:E_UI_BackplateBrightness Enum=NewEnumerator1) -> (then)
   Event OnMouseEnter()
   > Call Button Hovered!(execute:Exec, self:Options Menu Action Replay Object Reference) -> (then:Exec) {is=title K2Node_CallDelegate}
   > Call On Focused!(execute:Exec, self:Options Menu Action Replay Object Reference) -> (then:Exec) {is=title K2Node_CallDelegate}
   > PlayAnimationForward(InAnimation=Get OnHover() -> (OnHover:Widget Animation Object Reference))
   > SpawnSound2D!(
    execute
    Sound:Sound Base Object Reference=Get HoverSound() -> (HoverSound:Sound Base Object Reference)
    VolumeMultiplier:float=1
    PitchMultiplier:float=1
    StartTime:float=0
    ConcurrencySettings:Sound Concurrency Object Reference
    bPersistAcrossLevelTransition:bool=false
    bAutoDestroy:bool=true
   ) -> (then, ReturnValue:Audio Component Object Reference)
   > ShowSettingDescription!(
    execute
    self:Fort Replay Context Object Reference=Get FortReplayContext!() -> (ReturnValue:Fort Replay Context Object Reference) {is=title K2Node_GetBlueprintContextEx}
    Header:"Text (by ref)"=Get Title()
    Description:"Text (by ref)"=Get Description() -> (Description:Text)
   ) -> (then)
   Event OnMouseLeave!() -> (OutputDelegate, then, MouseEvent:"Pointer Event Structure (by ref)") {is=title K2Node_Event}
   > PlayAnimationReverse(InAnimation=Get OnHover())
   > Call Button Unhovered!(execute:Exec, self:Options Menu Action Replay Object Reference) -> (then:Exec) {is=title K2Node_CallDelegate}
  ]}
 Initialize {nodes [
   n1 = Initialize!() -> (then, TitleText:Text, ButtonText:Text, Formatting:E_Replay_TextFormat Enum=NewEnumerator2) {is=title K2Node_FunctionEntry}
   > Set Loc_TitleText(Loc_TitleText:Text=@n1) -> (Output_Get:Text)
   > Set Loc_ButtonText(Loc_ButtonText:Text=@n1.ButtonText) -> (Output_Get:Text)
   > n2 = Switch on E_Replay_TextFormat!(execute, Selection:E_Replay_TextFormat Enum=@n1.Formatting) -> (NewEnumerator2:Exec, NewEnumerator1:Exec, NewEnumerator3:Exec) {is=title K2Node_SwitchEnum}
   "SetText (Text)"(
    execute=@n2.NewEnumerator1
    self=Get Text_SettingName() -> (Text_SettingName:Common Text Object Reference)
    InText=Format Text!(
     Format="LOCTEXT(\"{indent once}{title}\")"
     indent once:Text=Get Loc_Indent() -> (Loc_Indent:Text)
     title:Text=Get Loc_TitleText() -> (Loc_TitleText:Text)
    ) -> (Result)
   )
   > UpdateButtonText!(execute, self, InputPin:Text=Get Loc_ButtonText() -> (Loc_ButtonText:Text)) -> (then)
   "SetText (Text)"(execute=@n2.NewEnumerator2, self=Get Text_SettingName(), InText=Get Loc_TitleText())
   > UpdateButtonText(InputPin=Get Loc_ButtonText())
   n3 = Get Loc_Indent()
   "SetText (Text)"(
    execute=@n2.NewEnumerator3
    self=Get Text_SettingName()
    InText=Format Text!(Format="LOCTEXT(\"{indent once}{indent twice}{title}\")", indent once=@n3, indent twice:Text=@n3, title=Get Loc_TitleText()) -> (Result)
   )
   > UpdateButtonText(InputPin=Get Loc_ButtonText())
  ]}
 OnFocusReceived {nodes [
   OnFocusReceived!() -> (then, MyGeometry:Geometry Structure, InFocusEvent:Focus Event Structure) {is=title K2Node_FunctionEntry}
   > CenterWidget!(execute, ToCenter:Widget Object Reference=Get Button_Action() -> (Button_Action:Icon Text Button Object Reference)) -> (then)
   > Return Node!(execute, ReturnValue:Event Reply Structure=Unhandled())
  ]}
 OnFocused {nodes [OnFocused() {is=title K2Node_FunctionEntry}]}
 UpdateButtonText {nodes [
   n1 = UpdateButtonText!() -> (then, InputPin:Text) {is=title K2Node_FunctionEntry}
   > Set Text!(execute, self:Icon Text Button Object Reference=Get Button_Action(), ButtonText:Text=@n1) -> (then) {is=title K2Node_CallFunction}
  ]}
]
variables [
 ButtonClicked {}
 ButtonHovered {}
 ButtonText {type=string, default="{Action}"}
 ButtonUnhovered {}
 Description {type=string, default=""}
 Formatting {type=E_Replay_TextFormat, default=None}
 HoverSound {type=SoundBase, default=/Game/Sounds/Fort_UI_Sounds/FrontEnd/DefaultSounds/PressedButtons/MSS_UI_Button_Hover_Default}
 OnFocused {}
 Title {type=string, default="{Setting Name}"}
]
widgets {
 info {inheritedWidgetCount=0, namedSlotCount=1, parentClass=&UserWidget, rootWidgetClass=&SizeBox, widgetCount=8}
 widgets [
  SizeBox_1 {widgetClassPath=&SizeBox, widget=SizeBox_1, properties {bOverride_WidthOverride=true, heightOverride=68, widthOverride=650}}
  Background {
   widgetClassPath=/FortniteCoreUI/Settings/SettingEntries/SettingEntryBackground
   bIsVariable=true
   parent=SizeBox_1
   widget=Background
   slot {class=SizeBoxSlot}
  }
  Overlay_0 {widgetClassPath=&Overlay, namedSlotHost=Background, widget=Overlay_0}
  FortResponsivePanel {
   widgetClassPath=/Script/FortniteUI.FortResponsivePanel
   parent=Overlay_0
   widget=FortResponsivePanel
   properties {bCanStackVertically=false}
   slot {class=OverlaySlot, properties {horizontalAlignment=HAlign_Fill, padding=<ltrb 2 2 2 2>}}
  }
  SizeBox_0 {
   widgetClassPath=&SizeBox
   parent=FortResponsivePanel
   widget=SizeBox_0
   properties {bOverride_MinDesiredHeight=true, minDesiredHeight=64}
   slot {class=FortResponsivePanelSlot}
  }
  Text_SettingName {
   widgetClassPath=&CommonTextBlock
   bIsVariable=true
   parent=SizeBox_0
   widget=Text_SettingName
   properties {
    scrollStyle=&TextScrollStyle-Base_C
    text="{Setting Name}"
    visibility=Visible
    colorAndOpacity {colorUseRule=UseColor_Specified, specifiedColor=<rgba 0.521 0.521 0.521 0.8>}
    font {
     bForceMonospaced=false
     bMaterialIsStencil=false
     fontMaterial=null
     fontObject=&HeadingNow
     letterSpacing=0
     monospacedWidth=1
     size=18
     skewAmount=0
     typefaceFontName=75 Medium
     outlineSettings {
      bApplyOutlineToDropShadows=false
      bMiteredCorners=false
      bSeparateFillAlpha=false
      outlineMaterial=null
      outlineSize=0
      outlineColor=<rgba 0 0 0 1>
     }
    }
    shadowOffset=<0 0>
   }
   slot {class=SizeBoxSlot, properties {horizontalAlignment=HAlign_Left, verticalAlignment=VAlign_Center, padding=<ltrb 24 0 10 0>}}
  }
  Button_Action {
   widgetClassPath=&IconTextButton
   bIsVariable=true
   parent=FortResponsivePanel
   widget=Button_Action
   properties {
    bInteractableWhenSelected=true
    bSelectable=true
    buttonText="{Action}"
    contentAlignment=HAlign_Center
    onButtonBaseClicked=()
    onButtonBaseDoubleClicked=()
    onButtonBaseHovered=()
    onButtonBaseUnhovered=()
    onSelectedChangedBase=()
    style=/FortniteCoreUI/Settings/SettingEntries/Styles/Settings_ButtonStyle_PurpleButton_Settings.Settings_ButtonStyle_PurpleButton_Settings_C
    touchMethod=PreciseTap
    font {
     bForceMonospaced=false
     bMaterialIsStencil=false
     fontMaterial=null
     fontObject=&HeadingNow
     letterSpacing=0
     monospacedWidth=1
     size=14.25
     skewAmount=0
     typefaceFontName=75 Medium
     outlineSettings {
      bApplyOutlineToDropShadows=false
      bMiteredCorners=false
      bSeparateFillAlpha=false
      outlineMaterial=null
      outlineSize=0
      outlineColor=<rgba 0 0 0 1>
     }
    }
   }
   slot {class=FortResponsivePanelSlot}
  }
  FocusOutline {
   widgetClassPath=&Image
   bIsVariable=true
   parent=Overlay_0
   widget=FocusOutline
   properties {
    pixelSnapping=Disabled
    renderOpacity=0
    visibility=SelfHitTestInvisible
    brush {
     bIsDynamicallyLoaded=false
     drawAs=Image
     imageType=NoImage
     mirroring=NoMirror
     resourceName=None
     resourceObject=/FortniteCoreUI/Settings/Art/MI_UIKit_Settings_Outline
     tiling=NoTile
     imageSize=<32 32>
     margin=<ltrb 0 0 0 0>
     outlineSettings {
      bUseBrushTransparency=false
      roundingType=HalfHeightRadius
      width=0
      color {colorUseRule=UseColor_Specified, specifiedColor=<rgba 0 0 0 0>}
      cornerRadii=<0 0 0 0>
     }
     tintColor {colorUseRule=UseColor_Specified, specifiedColor=<rgba 1 1 1 1>}
     uVRegion {bIsValid=false, max=<0 0>, min=<0 0>}
    }
   }
   slot {class=OverlaySlot, properties {horizontalAlignment=HAlign_Fill, verticalAlignment=VAlign_Fill, padding=<ltrb -16 -16 -16 -16>}}
  }
 ]
}
