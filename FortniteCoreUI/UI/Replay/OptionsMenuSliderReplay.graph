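# Replay options menu slider row showing a titled setting with a formatted value over a progress bar.
# An invisible normalized slider maps through MinMaxValue to formatted text, updates the bar and fires change and end delegates.
# An optional checkbox, non-shipping builds only, swaps the slider for typed entry, clamped to range unless bOverrideEditable is set.
# Hover plays a highlight and sound, fires On Focused and sends the description to the replay context.
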WidgetBlueprint /FortniteCoreUI/UI/Replay/OptionsMenuSliderReplay
animations [OnHover {
  binding_count=3
  length_seconds=0.1
  object=&self:OnHover
  sequence {bindings [
    Text_SettingValue {
     id=#1
     possessed_class=CommonTextBlock
     tracks [{
       type=MovieSceneColorTrack
       display_name=Color and Opacity
       sections [{
         type=MovieSceneColorSection
         end_frame=2
         start_frame=0
         channels [
          Color.R {type=MovieSceneScriptingFloatChannel, keys [[0,0.521],[2,1]]}
          Color.G {type=MovieSceneScriptingFloatChannel, keys [[0,0.521],[2,1]]}
          Color.B {type=MovieSceneScriptingFloatChannel, keys [[0,0.521],[2,1]]}
          Color.A {type=MovieSceneScriptingFloatChannel, keys [[0,0.8],[2,1]]}
         ]
        }]
      }]
    }
    Text_SettingName {
     id=#2
     possessed_class=CommonTextBlock
     tracks [{
       type=MovieSceneColorTrack
       display_name=Color and Opacity
       sections [{
         type=MovieSceneColorSection
         end_frame=2
         start_frame=0
         channels [
          Color.R {type=MovieSceneScriptingFloatChannel, keys [[0,0.521],[2,1]]}
          Color.G {type=MovieSceneScriptingFloatChannel, keys [[0,0.521],[2,1]]}
          Color.B {type=MovieSceneScriptingFloatChannel, keys [[0,0.521],[2,1]]}
          Color.A {type=MovieSceneScriptingFloatChannel, keys [[0,0.8],[2,1]]}
         ]
        }]
      }]
    }
    FocusOutline {
     id=#3
     possessed_class=Image
     tracks [{
       type=MovieSceneFloatTrack
       display_name=Render Opacity
       sections [{type=MovieSceneFloatSection, end_frame=2, start_frame=0, channels [None {type=MovieSceneScriptingFloatChannel, keys [[0,0],[2,1]]}]}]
      }]
    }
   ]}
 }]
graphs [
 EventGraph {nodes [
   n1 = "On Selected Changed (CheckBox)"!() -> (OutputDelegate, then, Button:Common Button Base Object Reference, Selected:bool) {is=title K2Node_ComponentBoundEvent}
   > n2 = Branch(Condition=@n1.Selected)
   > "SetText (Editable Text)"!(execute, self:Editable Text Object Reference=Get Text_Editable() -> (Text_Editable:Editable Text Object Reference), InText:Text) -> (then)
   > SetVisibility(self=Get Text_Editable(), InVisibility=Visible)
   > SetVisibility(
    self=[Get Slider() -> (Slider:Analog Slider Object Reference), Get Text_SettingValue() -> (Text_SettingValue:Common Text Object Reference)]
    InVisibility=Collapsed
   )
   > SetKeyboardFocus(self:Widget Object Reference=Get Text_Editable())
   SetVisibility(execute=@n2.else, self=Get Text_Editable(), InVisibility=Collapsed)
   > SetVisibility(self=Get Text_SettingValue(), InVisibility=SelfHitTestInvisible)
   > SetVisibility(self=Get Slider(), InVisibility=Visible)
   n3 = "OnTextCommitted (Editable Text) (Text_Editable)"!() -> (OutputDelegate, then, Text:"Text (by ref)", CommitMethod:ETextCommit Enum=Default) {is=title K2Node_ComponentBoundEvent}
   n4 = Get Text_Editable()
   > n5 = Branch(Condition=TextIsEmpty(InText="GetText (Editable Text)"!(self:Editable Text Object Reference=@n4) -> (ReturnValue:Text)))
   n6 = Branch(execute=@n5.else, Condition=IsMobileGame())
   > n7 = Switch on ETextCommit!(execute, Selection:ETextCommit Enum=@n3.CommitMethod) -> (Default:Exec, OnEnter:Exec, OnUserMovedFocus:Exec, OnCleared:Exec) {is=title K2Node_SwitchEnum}
   n8 = Branch(execute=[@n16.OnEnter, @n7.Default, @n7.OnEnter, @n7.OnUserMovedFocus], Condition=Get bOverrideEditable() -> (bOverrideEditable:bool))
   n9 = Text to Float!(InText:"Text (by ref)"=@n3.Text) -> (ReturnValue:double=~) {is=title K2Node_MacroInstance}
   n10 = Format Text!(
    Format="INVTEXT(\"{prefix}{value}{suffix}\")"
    prefix:Text=Get ValuePrefixText() -> (ValuePrefixText:Text)
    value:Text="To Text (Float)"(
     Value=@n9
     bUseGrouping=Get bGroupEditableText() -> (bGroupEditableText:bool)
     MinimumIntegralDigits=Get MinIntegralDigits() -> (MinIntegralDigits:int=0)
     MaximumIntegralDigits=Get MaxIntegralDigits() -> (MaxIntegralDigits:int=0)
     MinimumFractionalDigits=Get MinFractionalDigits() -> (MinFractionalDigits:int=0)
     MaximumFractionalDigits=Get MaxFractionalDigits() -> (MaxFractionalDigits:int=0)
    )
    suffix:Text=Get ValueSuffixText() -> (ValueSuffixText:Text)
   ) -> (Result)
   > "SetText (Text)"(self=Get Text_SettingValue(), InText=@n10)
   > SetHintText!(execute, self:Editable Text Object Reference=Get Text_Editable(), InHintText:Text=@n10) -> (then)
   > "SetText (Editable Text)"(self=Get Text_Editable(), InText=@n10)
   n11 = Get!() -> (MinMaxValue_X:double, MinMaxValue_Y:double)
   > n12 = Branch(Condition="float > float"(A=@n9, B=@n11.MinMaxValue_Y))
   > SetProgress!(
    execute
    self:Fort Simple Material Progress Bar Object Reference=Get Progress() -> (Progress:Simple Material Progress Bar Object Reference)
    Progress:float=float / float(A=@n9, B=@n9)
   ) -> (then)
   > Call On Value Commited!(execute:Exec, self:Options Menu Slider Replay Object Reference, Value:double=@n9) -> (then:Exec) {is=title K2Node_CallDelegate}
   > SetKeyboardFocus(self=Get Text_Editable())
   SetProgress(execute=@n12.else, self=Get Progress(), Progress=float / float(A=@n9, B=@n11.MinMaxValue_Y))
   > Call On Value Commited(Value=@n9)
   > SetKeyboardFocus(self=Get Text_Editable())
   n13 = Get()
   n14 = "Clamp (Float)"(Value=@n9, Min=@n13, Max=@n13.MinMaxValue_Y)
   n15 = Format Text(
    Format="INVTEXT(\"{prefix}{value}{suffix}\")"
    prefix=Get ValuePrefixText()
    value="To Text (Float)"(
     Value=@n14
     bUseGrouping=Get bGroupEditableText()
     MinimumIntegralDigits=Get MinIntegralDigits() -> (MinIntegralDigits=0)
     MaximumIntegralDigits=Get MaxIntegralDigits() -> (MaxIntegralDigits=0)
     MinimumFractionalDigits=Get MinFractionalDigits() -> (MinFractionalDigits=0)
     MaximumFractionalDigits=Get MaxFractionalDigits() -> (MaxFractionalDigits=0)
    )
    suffix=Get ValueSuffixText()
   )
   "SetText (Text)"(execute=@n8.else, self=Get Text_SettingValue(), InText=@n15)
   > SetHintText(self=Get Text_Editable(), InHintText=@n15)
   > "SetText (Editable Text)"(self=Get Text_Editable(), InText=@n15)
   > SetProgress(self=Get Progress(), Progress=float / float(A=@n14, B=Get().MinMaxValue_Y))
   > Call On Value Commited(Value=@n14)
   > SetKeyboardFocus(self=Get Text_Editable())
   n16 = Switch on ETextCommit(execute=@n6.else, Selection=@n3.CommitMethod)
   "On Controller Capture End (Slider)"() {is=title K2Node_ComponentBoundEvent}
   n17 = Get Slider()
   n18 = GetValue!(self:Slider Object Reference=@n17) -> (ReturnValue:float)
   > Call On Slider End!(execute:Exec, self:Options Menu Slider Replay Object Reference, SliderValue:double=@n18) -> (then:Exec) {is=title K2Node_CallDelegate}
   > Call Slider Changed!(execute:Exec, self:Options Menu Slider Replay Object Reference, Slider Value:double=GetValue(self=@n17)) -> (then:Exec) {is=title K2Node_CallDelegate}
   n19 = "On Value Changed (Slider)"!() -> (OutputDelegate, then, Value:float) {is=title K2Node_ComponentBoundEvent}
   n20 = GetSliderValueText!(Value:double=@n19.Value) -> (ValueText:Text)
   > n21 = "SetText (Text)"(self=Get Text_SettingValue(), InText=@n20)
   SetHintText(execute=[@n21.then, @n26.then, @n29.then], self=Get Text_Editable(), InHintText=@n20)
   n22 = GetValue(self=Get Slider())
   > SetProgress(self=Get Progress(), Progress=@n22)
   > Call Slider Changed(Slider Value=@n22)
   "On Mouse Capture End (Slider)"() {is=title K2Node_ComponentBoundEvent}
   > Call On Slider End(SliderValue=@n18)
   > Call Slider Changed(Slider Value=@n18)
   n23 = "On Analog Capture (Slider)"!() -> (OutputDelegate, then, Value:float) {is=title K2Node_ComponentBoundEvent}
   > Call Slider Changed(Slider Value=@n23.Value)
   > PlaySound2D(Sound=Get SliderSound() -> (SliderSound:Sound Base Object Reference))
   ForceSliderRight() {is=title K2Node_CustomEvent}
   n24 = Get Slider()
   n25 = "Clamp (Float)"(
    Value=float + float(A=Get StepSize!(self:Slider Object Reference=@n24) -> (StepSize:float), B=Get Value!(self:Slider Object Reference=@n24) -> (Value:float))
    Min=Get MinValue!(self:Slider Object Reference=@n24) -> (MinValue:float)
    Max=Get MaxValue!(self:Slider Object Reference=@n24) -> (MaxValue:float)
   )
   > SetValue!(execute, self:Slider Object Reference=@n24, InValue:float=@n25) -> (then)
   > n26 = "SetText (Text)"(self=Get Text_SettingValue(), InText=GetSliderValueText(Value=@n25))
   ForceSliderLeft() {is=title K2Node_CustomEvent}
   n27 = Get Slider()
   n28 = "Clamp (Float)"(Value=float - float(A=Get Value(self=@n27), B=Get StepSize(self=@n27)), Min=Get MinValue(self=@n27), Max=Get MaxValue(self=@n27))
   > SetValue(self=@n27, InValue=@n28)
   > n29 = "SetText (Text)"(self=Get Text_SettingValue(), InText=GetSliderValueText(Value=@n28))
   Event PreConstruct()
   > Update Slider!(
    execute
    self
    Slider Text:Text=Get Title() -> (Title:Text)
    Slider Value:double=Get Slider Value() -> (Slider Value:double)
    Hover Text:Text
    Tooltip Text Block:Common Text Object Reference
    Min Max Value:Vector 2D Structure=Get MinMaxValue() -> (MinMaxValue:Vector 2D Structure)
   ) -> (then)
   n30 = Get WBP_UIKit_Backplate!(self:Setting Entry Background Object Reference=Get Background() -> (Background:Setting Entry Background Object Reference)) -> (WBP_UIKit_Backplate:WBP UIKit Backplate Object Reference)
   > SetBackplateMaterial!(
    execute
    self:WBP UIKit Backplate Base Object Reference=@n30
    Material Instance:Material Instance Object Reference=/FortniteCoreUI/UI/Replay/Materials/M_Replay_Backplate_Inst
   ) -> (then)
   > SetBrightness!(execute, self:WBP UIKit Backplate Base Object Reference=@n30, Brightness:E_UI_BackplateBrightness Enum=NewEnumerator1) -> (then)
   n31 = GetRuntimeOptions()
   > n32 = Branch(Condition=NOT Boolean(A=IsShippingBuild!(self:Fort Runtime Options Object Reference=@n31) -> (ReturnValue:bool)))
   > n33 = Branch(Condition=Get bUseCheckBox() -> (bUseCheckBox:bool))
   > SetVisibility(
    self=[
     Get CheckBox() -> (CheckBox:WBP UIKit Button With Toggle Object Reference)
     Get SizeBoxCheckBox() -> (SizeBoxCheckBox:Size Box Object Reference)
    ]
    InVisibility=SelfHitTestInvisible
   )
   > SetWidthOverride(self=Get SizeBoxSettingName() -> (SizeBoxSettingName:Size Box Object Reference), InWidthOverride=248.5)
   > SetVisibility(self=Get Text_Editable(), InVisibility=Collapsed)
   > SetHintText(self=Get Text_Editable(), InHintText=GetSliderValueText(Value=Get Slider Value()))
   > n34 = Branch(Condition=Get bOverrideEditable())
   n35 = Get CheckBox()
   > SetColorAndOpacity(
    self:Image Object Reference=Get Image!(self:WBP UIKit Block Toggle Object Reference=Get Block_Toggle!(self:WBP UIKit Button With Toggle Object Reference=@n35) -> (Block_Toggle:WBP UIKit Block Toggle Object Reference)) -> (Image:Fort Mobile Image Object Reference)
    InColorAndOpacity:Linear Color Structure=(R=0.9473,G=0.0284,B=0.0437,A=1)
   )
   n36 = Get CheckBox()
   SetColorAndOpacity(execute=@n34.else, self=Get Image(self=Get Block_Toggle(self=@n36)), InColorAndOpacity=(R=1,G=1,B=1,A=1))
   SetVisibility(execute=[@n33.else, @n32.else], self=[Get CheckBox(), Get SizeBoxCheckBox(), Get Text_Editable()], InVisibility=Collapsed)
   > SetWidthOverride(self=Get SizeBoxSettingName(), InWidthOverride=312.5)
   > n37 = Branch(Condition=Get bUseCustomWidth() -> (bUseCustomWidth:bool))
   > SetWidthOverride(self=Get SizeBoxSettingName(), InWidthOverride=Get CustomWidth() -> (CustomWidth:float))
   SetWidthOverride(execute=@n37.else, self=Get SizeBoxSettingName(), InWidthOverride=312.5)
   Event OnMouseLeave!() -> (OutputDelegate, then, MouseEvent:"Pointer Event Structure (by ref)") {is=title K2Node_Event}
   > "SetText (Text)"(self=Get Tab Tooltip Text() -> (Tab Tooltip Text:Common Text Object Reference))
   > PlayAnimationReverse(InAnimation=Get OnHover() -> (OnHover:Widget Animation Object Reference))
   Event OnMouseEnter()
   > "SetText (Text)"(self=Get Tab Tooltip Text(), InText=Get Hover Text() -> (Hover Text:Text))
   > Call On Focused!(execute:Exec, self:Options Menu Slider Replay Object Reference) -> (then:Exec) {is=title K2Node_CallDelegate}
   > PlayAnimationForward(InAnimation=Get OnHover())
   > SpawnSound2D!(
    execute
    Sound:Sound Base Object Reference=Get HoverSound() -> (HoverSound:Sound Base Object Reference)
    VolumeMultiplier:float=1
    PitchMultiplier:float=1
    StartTime:float=0
    ConcurrencySettings:Sound Concurrency Object Reference
    bPersistAcrossLevelTransition:bool=false
    bAutoDestroy:bool=true
   ) -> (then, ReturnValue:Audio Component Object Reference)
   > ShowSettingDescription!(
    execute
    self:Fort Replay Context Object Reference=Get FortReplayContext!() -> (ReturnValue:Fort Replay Context Object Reference) {is=title K2Node_GetBlueprintContextEx}
    Header:"Text (by ref)"=Get Title()
    Description:"Text (by ref)"=Get Description() -> (Description:Text)
   ) -> (then)
   Branch(Condition=HasMouseCapture!(self:Widget Object Reference=Get Slider()) -> (ReturnValue:bool))
   > Branch(Condition=Get MouseDragUpdates() -> (MouseDragUpdates:bool))
  ]}
 FormatText {nodes [
   n1 = FormatText!() -> (then, TitleIn:Text) {is=title K2Node_FunctionEntry}
   > Set Loc_TextIn(Loc_TextIn:Text=@n1) -> (Output_Get:Text)
   > n2 = Switch on E_Replay_TextFormat!(execute, Selection:E_Replay_TextFormat Enum=Get Formatting() -> (Formatting:E_Replay_TextFormat Enum=NewEnumerator2)) -> (NewEnumerator2:Exec, NewEnumerator1:Exec, NewEnumerator3:Exec) {is=title K2Node_SwitchEnum}
   Return Node!(
    execute=@n2.NewEnumerator1
    TextOut:Text=Format Text!(
     Format="LOCTEXT(\"{indent once}{title}\")"
     indent once:Text=Get Loc_Indent() -> (Loc_Indent:Text)
     title:Text=Get Loc_TextIn() -> (Loc_TextIn:Text)
    ) -> (Result)
   )
   > Return Node(TextOut=Get Loc_TextIn())
   n3 = Get Loc_Indent()
   Return Node(
    execute=@n2.NewEnumerator3
    TextOut=Format Text!(Format="LOCTEXT(\"{indent once}{indent twice}{title}\")", indent once=@n3, indent twice:Text=@n3, title=Get Loc_TextIn()) -> (Result)
   )
  ]}
 GetMappedValue {nodes [
   GetMappedValue() {is=title K2Node_FunctionEntry}
   n1 = Get Slider()
   n2 = Get()
   > Return Node!(execute, MappedValue:double=MapRangeClamped(Value=GetValue(self=@n1), InRangeA=0, InRangeB=1, OutRangeA=@n2, OutRangeB=@n2.MinMaxValue_Y))
  ]}
 GetSliderValueText {nodes [
   n1 = GetSliderValueText!() -> (then, Value:double=~) {is=title K2Node_FunctionEntry}
   n2 = Get()
   n3 = MapRangeClamped(Value=@n1, InRangeA=0, InRangeB=1, OutRangeA=@n2, OutRangeB=@n2.MinMaxValue_Y)
   n4 = Get MinIntegralDigits() -> (MinIntegralDigits=0)
   n5 = Get MaxIntegralDigits() -> (MaxIntegralDigits=0)
   n6 = Get MinFractionalDigits() -> (MinFractionalDigits=0)
   n7 = Get MaxFractionalDigits() -> (MaxFractionalDigits=0)
   > Return Node!(
    execute
    ValueText:Text=Format Text!(
     Format="INVTEXT(\"{prefix}{value}{suffix}\")"
     prefix:Text=Get ValuePrefixText()
     value:Text="To Text (Float)"(Value=@n3, MinimumIntegralDigits=@n4, MaximumIntegralDigits=@n5, MinimumFractionalDigits=@n6, MaximumFractionalDigits=@n7)
     suffix:Text=Get ValueSuffixText()
    ) -> (Result)
   )
  ]}
 On Focused {nodes [On Focused() {is=title K2Node_FunctionEntry}]}
 OnFocusReceived {nodes [
   OnFocusReceived!() -> (then, MyGeometry:Geometry Structure, InFocusEvent:Focus Event Structure) {is=title K2Node_FunctionEntry}
   > CenterWidget!(execute, ToCenter:Widget Object Reference=Get Slider()) -> (then)
   > Return Node!(execute, ReturnValue:Event Reply Structure=Handled())
  ]}
 OnSliderEnd {nodes [OnSliderEnd!() -> (then, SliderValue:double=~) {is=title K2Node_FunctionEntry}]}
 OnValueCommited {nodes [OnValueCommited!() -> (then, Value:double=~) {is=title K2Node_FunctionEntry}]}
 SliderChanged {nodes [SliderChanged!() -> (then, Slider Value:double=~) {is=title K2Node_FunctionEntry}]}
 TextToFloat {nodes [
   n1 = Inputs!() -> (InText:"Text (by ref)")
   Outputs!(ReturnValue:double=String To Float(InString="To String (Text)"(InText=@n1)))
  ]}
 Update Slider {nodes [
   n1 = Update Slider!() -> (
    then
    Slider Text:Text
    Slider Value:double=~
    Hover Text:Text
    Tooltip Text Block:Common Text Object Reference
    Min Max Value:Vector 2D Structure
   ) {is=title K2Node_FunctionEntry}
   > n2 = Set Title(Title:Text=@n1) -> (Output_Get:Text)
   > "SetText (Text)"(
    self=Get Text_SettingName() -> (Text_SettingName:Common Text Object Reference)
    InText=FormatText!(TitleIn:Text=@n2) -> (TextOut:Text) {is=title K2Node_CallFunction}
   )
   > Set Hover Text(Hover Text:Text=@n1.Hover Text) -> (Output_Get:Text)
   > Set Tab Tooltip Text(Tab Tooltip Text:Common Text Object Reference=@n1.Tooltip Text Block) -> (Output_Get:Common Text Object Reference)
   > Set MinMaxValue(MinMaxValue:Vector 2D Structure=@n1.Min Max Value) -> (Output_Get:Vector 2D Structure)
   > Update Slider Value!(execute, self, NormalizedValue:double=@n1.Slider Value) -> (then)
  ]}
 Update Slider Value {nodes [
   n1 = Update Slider Value!() -> (then, NormalizedValue:double=~) {is=title K2Node_FunctionEntry}
   > "SetText (Text)"(self=Get Text_SettingValue(), InText=GetSliderValueText(Value=@n1) {is=title K2Node_CallFunction})
   > SetValue(self=Get Slider(), InValue=@n1)
   > SetProgress(self=Get Progress(), Progress=@n1)
  ]}
 UpdateMinMaxValue {nodes [
   n1 = UpdateMinMaxValue!() -> (then, MinMaxValue:Vector 2D Structure) {is=title K2Node_FunctionEntry}
   > Set MinMaxValue(MinMaxValue=@n1)
  ]}
 UpdateSliderStep {nodes [
   n1 = UpdateSliderStep!() -> (then, Step:double=~) {is=title K2Node_FunctionEntry}
   n2 = Get()
   n3 = float - float(A=@n2, B=@n2.MinMaxValue_Y)
   > SetStepSize!(execute, self:Slider Object Reference=Get Slider(), InValue:float=float / float(A=@n1, B="Absolute (Float)"(A=@n3))) -> (then)
  ]}
 UpdateSliderValueUnmapped {nodes [
   n1 = UpdateSliderValueUnmapped!() -> (then, UnmappedValue:double=~) {is=title K2Node_FunctionEntry}
   n2 = Get()
   > Update Slider Value(NormalizedValue=NormalizeToRange!(Value:double=@n1, RangeMin:double=@n2, RangeMax:double=@n2.MinMaxValue_Y) -> (ReturnValue:double)) {is=title K2Node_CallFunction}
  ]}
]
variables [
 bGroupEditableText {type=boolean, default=true}
 bOverrideEditable {type=boolean}
 bUseCheckBox {type=boolean}
 bUseCustomWidth {type=boolean}
 CustomWidth {type=number, default=287.5}
 Description {type=string, default=""}
 Formatting {type=E_Replay_TextFormat, default=None}
 Hover Text {type=string, default=""}
 HoverSound {type=SoundBase, default=/Game/Sounds/Fort_UI_Sounds/FrontEnd/DefaultSounds/PressedButtons/MSS_UI_Button_Hover_Default}
 MaxFractionalDigits {type=integer, default=2}
 MaxIntegralDigits {type=integer, default=1}
 MinFractionalDigits {type=integer, default=2}
 MinIntegralDigits {type=integer, default=1}
 MinMaxValue {type=Vector2D, default=<0 1>}
 MouseDragUpdates {type=boolean, default=true}
 On Focused {}
 OnSliderEnd {}
 OnValueCommited {}
 Slider Value {type=number, default=1}
 SliderChanged {}
 SliderSound {type=SoundBase, default=/Game/Sounds/Fort_UI_Sounds/Replay/DVR_CamOptions_Slider_Cue}
 Tab Tooltip Text {type=CommonTextBlock, default=null}
 Title {type=string, default="{Setting Name}"}
 ValuePrefixText {type=string, default=""}
 ValueSuffixText {type=string, default=""}
]
widgets {
 info {inheritedWidgetCount=0, namedSlotCount=1, parentClass=&UserWidget, rootWidgetClass=&SizeBox, widgetCount=17}
 widgets [
  SizeBoxMaster {widgetClassPath=&SizeBox, bIsVariable=true, widget=SizeBoxMaster, properties {bOverride_WidthOverride=true, widthOverride=650}}
  Overlay_0 {widgetClassPath=&Overlay, parent=SizeBoxMaster, widget=Overlay_0, slot {class=SizeBoxSlot}}
  Background {
   widgetClassPath=/FortniteCoreUI/Settings/SettingEntries/SettingEntryBackground
   bIsVariable=true
   parent=Overlay_0
   widget=Background
   slot {class=OverlaySlot, properties {horizontalAlignment=HAlign_Fill, verticalAlignment=VAlign_Fill}}
  }
  Overlay {widgetClassPath=&Overlay, namedSlotHost=Background, widget=Overlay}
  FortResponsivePanel_0 {
   widgetClassPath=/Script/FortniteUI.FortResponsivePanel
   parent=Overlay
   widget=FortResponsivePanel_0
   properties {bCanStackVertically=false, bOverride_Cursor=true}
   slot {class=OverlaySlot, properties {horizontalAlignment=HAlign_Fill}}
  }
  HorizontalBox_0 {widgetClassPath=&HorizontalBox, parent=FortResponsivePanel_0, widget=HorizontalBox_0, slot {class=FortResponsivePanelSlot}}
  SizeBoxSettingName {
   widgetClassPath=&SizeBox
   bIsVariable=true
   parent=HorizontalBox_0
   widget=SizeBoxSettingName
   properties {bOverride_MinDesiredHeight=true, bOverride_WidthOverride=true, minDesiredHeight=64, widthOverride=297.7778}
   slot {class=HorizontalBoxSlot}
  }
  Text_SettingName {
   widgetClassPath=&CommonTextBlock
   bIsVariable=true
   parent=SizeBoxSettingName
   widget=Text_SettingName
   properties {
    scrollStyle=&TextScrollStyle-Base_C
    text="{Setting Name}"
    visibility=Visible
    colorAndOpacity {colorUseRule=UseColor_Specified, specifiedColor=<rgba 0.521 0.521 0.521 0.8>}
    font {
     bForceMonospaced=false
     bMaterialIsStencil=false
     fontMaterial=null
     fontObject=&HeadingNow
     letterSpacing=0
     monospacedWidth=1
     size=18
     skewAmount=0
     typefaceFontName=75 Medium
     outlineSettings {
      bApplyOutlineToDropShadows=false
      bMiteredCorners=false
      bSeparateFillAlpha=false
      outlineMaterial=null
      outlineSize=0
      outlineColor=<rgba 0 0 0 1>
     }
    }
    shadowOffset=<0 0>
   }
   slot {class=SizeBoxSlot, properties {horizontalAlignment=HAlign_Left, verticalAlignment=VAlign_Center, padding=<ltrb 24 0 10 0>}}
  }
  SizeBoxCheckBox {
   widgetClassPath=&SizeBox
   bIsVariable=true
   parent=HorizontalBox_0
   widget=SizeBoxCheckBox
   properties {bOverride_WidthOverride=true, widthOverride=64}
   slot {class=HorizontalBoxSlot}
  }
  CheckBox {
   widgetClassPath=/FortUILibrary/UIKit/Buttons/WBP_UIKit_ButtonWithToggle
   bIsVariable=true
   parent=SizeBoxCheckBox
   widget=CheckBox
   properties {toggleType=CheckBox, useTextBlock=false, buttonPadding=<ltrb 0 0 0 0>}
   slot {class=SizeBoxSlot, properties {horizontalAlignment=HAlign_Center, verticalAlignment=VAlign_Center}}
  }
  SizeBoxValue {
   widgetClassPath=&SizeBox
   parent=HorizontalBox_0
   widget=SizeBoxValue
   properties {bOverride_MinDesiredHeight=true, minDesiredHeight=50}
   slot {class=HorizontalBoxSlot, properties {size {sizeRule=Fill, value=1}}}
  }
  PanelValue {widgetClassPath=&Overlay, parent=SizeBoxValue, widget=PanelValue, slot {class=SizeBoxSlot}}
  Progress {
   widgetClassPath=/Game/Athena/HUD/SimpleMaterialProgressBar
   bIsVariable=true
   parent=PanelValue
   widget=Progress
   properties {
    bOverrideDefaultColorA=true
    bOverrideDefaultColorB=true
    bOverrideDefaultColorBackground=true
    visibility=HitTestInvisible
    cachedColorA=<rgba 0 0.0232 0.2502 1>
    cachedColorB=<rgba 0 0.0232 0.2502 1>
    cachedColorBackground=<rgba 0 0 0.0331 1>
   }
   slot {class=OverlaySlot, properties {horizontalAlignment=HAlign_Fill, verticalAlignment=VAlign_Fill}}
  }
  Text_SettingValue {
   widgetClassPath=&CommonTextBlock
   bIsVariable=true
   parent=PanelValue
   widget=Text_SettingValue
   properties {
    justification=Center
    scrollStyle=&TextScrollStyle-Base_C
    text=1.0
    visibility=HitTestInvisible
    colorAndOpacity {colorUseRule=UseColor_Specified, specifiedColor=<rgba 0.521 0.521 0.521 0.8>}
    font {
     bForceMonospaced=false
     bMaterialIsStencil=false
     fontMaterial=null
     fontObject=&HeadingNow
     letterSpacing=0
     monospacedWidth=1
     size=18
     skewAmount=0
     typefaceFontName=74 Regular
     outlineSettings {
      bApplyOutlineToDropShadows=false
      bMiteredCorners=false
      bSeparateFillAlpha=false
      outlineMaterial=null
      outlineSize=0
      outlineColor=<rgba 0 0 0 1>
     }
    }
    shadowOffset=<0 0>
   }
   slot {class=OverlaySlot, properties {horizontalAlignment=HAlign_Fill, verticalAlignment=VAlign_Center}}
  }
  Slider {
   widgetClassPath=/Script/CommonUI.AnalogSlider
   bIsVariable=true
   parent=PanelValue
   widget=Slider
   properties {value=1, sliderBarColor=<rgba 1 1 1 0>, sliderHandleColor=<rgba 1 1 1 0>}
   slot {class=OverlaySlot, properties {horizontalAlignment=HAlign_Fill, verticalAlignment=VAlign_Fill}}
  }
  Text_Editable {
   widgetClassPath=/Script/UMG.EditableText
   bIsVariable=true
   parent=PanelValue
   widget=Text_Editable
   properties {
    hintText="{Editable Text}"
    justification=Center
    widgetStyle {
     backgroundImageComposing {
      bIsDynamicallyLoaded=false
      drawAs=Border
      imageType=FullColor
      mirroring=NoMirror
      resourceName=../../../Engine/Content/Slate/Old/HyperlinkDotted.png
      resourceObject=null
      tiling=Both
      imageSize=<0 0>
      margin=<ltrb 0 0 0 0.1875>
      outlineSettings {
       bUseBrushTransparency=false
       roundingType=HalfHeightRadius
       width=0
       color {colorUseRule=UseColor_Specified, specifiedColor=<rgba 0 0 0 0>}
       cornerRadii=<0 0 0 0>
      }
      tintColor {colorUseRule=UseColor_Specified, specifiedColor=<rgba 1 1 1 1>}
      uVRegion {bIsValid=false, max=<0 0>, min=<0 0>}
     }
     backgroundImageSelected {
      bIsDynamicallyLoaded=false
      drawAs=Box
      imageType=FullColor
      mirroring=NoMirror
      resourceName=../../../Engine/Content/Slate/Common/EditableTextSelectionBackground.png
      resourceObject=null
      tiling=NoTile
      imageSize=<0 0>
      margin=<ltrb 0.25 0.25 0.25 0.25>
      outlineSettings {
       bUseBrushTransparency=false
       roundingType=HalfHeightRadius
       width=0
       color {colorUseRule=UseColor_Specified, specifiedColor=<rgba 0 0 0 0>}
       cornerRadii=<0 0 0 0>
      }
      tintColor {colorUseRule=UseColor_Specified, specifiedColor=<rgba 1 1 1 1>}
      uVRegion {bIsValid=false, max=<0 0>, min=<0 0>}
     }
     caretImage {
      bIsDynamicallyLoaded=false
      drawAs=Image
      imageType=FullColor
      mirroring=NoMirror
      resourceName=../../../Engine/Content/Slate/Old/White.png
      resourceObject=null
      tiling=NoTile
      imageSize=<16 16>
      margin=<ltrb 0 0 0 0>
      outlineSettings {
       bUseBrushTransparency=false
       roundingType=HalfHeightRadius
       width=0
       color {colorUseRule=UseColor_Specified, specifiedColor=<rgba 0 0 0 0>}
       cornerRadii=<0 0 0 0>
      }
      tintColor {colorUseRule=UseColor_Specified, specifiedColor=<rgba 1 1 1 1>}
      uVRegion {bIsValid=false, max=<0 0>, min=<0 0>}
     }
     colorAndOpacity {colorUseRule=UseColor_Specified, specifiedColor=<rgba 1 1 1 1>}
     font {
      bForceMonospaced=false
      bMaterialIsStencil=false
      fontMaterial=null
      fontObject=&HeadingNow
      letterSpacing=0
      monospacedWidth=1
      size=18
      skewAmount=0
      typefaceFontName=74 Regular
      outlineSettings {
       bApplyOutlineToDropShadows=false
       bMiteredCorners=false
       bSeparateFillAlpha=false
       outlineMaterial=null
       outlineSize=0
       outlineColor=<rgba 0 0 0 1>
      }
     }
     hintColorAndOpacity {colorUseRule=UseColor_Foreground, specifiedColor=<rgba 1 0 1 1>}
     selectedBackgroundColor {colorUseRule=UseColor_Foreground, specifiedColor=<rgba 1 0 1 1>}
    }
   }
   slot {class=OverlaySlot, properties {horizontalAlignment=HAlign_Fill, verticalAlignment=VAlign_Center}}
  }
  FocusOutline {
   widgetClassPath=&Image
   bIsVariable=true
   parent=Overlay
   widget=FocusOutline
   properties {
    pixelSnapping=Disabled
    renderOpacity=0
    visibility=SelfHitTestInvisible
    brush {
     bIsDynamicallyLoaded=false
     drawAs=Image
     imageType=NoImage
     mirroring=NoMirror
     resourceName=None
     resourceObject=/FortniteCoreUI/Settings/Art/MI_UIKit_Settings_Outline
     tiling=NoTile
     imageSize=<32 32>
     margin=<ltrb 0 0 0 0>
     outlineSettings {
      bUseBrushTransparency=false
      roundingType=HalfHeightRadius
      width=0
      color {colorUseRule=UseColor_Specified, specifiedColor=<rgba 0 0 0 0>}
      cornerRadii=<0 0 0 0>
     }
     tintColor {colorUseRule=UseColor_Specified, specifiedColor=<rgba 1 1 1 1>}
     uVRegion {bIsValid=false, max=<0 0>, min=<0 0>}
    }
   }
   slot {class=OverlaySlot, properties {horizontalAlignment=HAlign_Fill, verticalAlignment=VAlign_Fill, padding=<ltrb -16 -16 -16 -16>}}
  }
 ]
}
