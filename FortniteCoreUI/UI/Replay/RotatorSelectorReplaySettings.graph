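# Replay settings row widget that shows a titled, optionally indented setting with a TextRotator cycling through text options via chevron buttons or forced-rotate events.
# Initialize and update functions fill labels and push option count, selected and default indices into the rotator's dynamic material.
# Selection changes fire a delegate and play a sound.
# Mouse hover plays a highlight animation, sets tooltip text and shows the description through FortReplayContext; focus centers the row.
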WidgetBlueprint /FortniteCoreUI/UI/Replay/RotatorSelectorReplaySettings
animations [OnHover {
  binding_count=4
  length_seconds=0.1
  object=&self:OnHover
  sequence {bindings [
    Text_SettingName {
     id=#1
     possessed_class=CommonTextBlock
     tracks [{
       type=MovieSceneColorTrack
       display_name=Color and Opacity
       sections [{
         type=MovieSceneColorSection
         end_frame=2
         start_frame=0
         channels [
          Color.R {type=MovieSceneScriptingFloatChannel, keys [[0,0.521],[2,1]]}
          Color.G {type=MovieSceneScriptingFloatChannel, keys [[0,0.521],[2,1]]}
          Color.B {type=MovieSceneScriptingFloatChannel, keys [[0,0.521],[2,1]]}
          Color.A {type=MovieSceneScriptingFloatChannel, keys [[0,0.8],[2,1]]}
         ]
        }]
      }]
    }
    Button_Decrease {
     id=#2
     possessed_class=IconTextButton_C
     tracks [{
       type=MovieSceneColorTrack
       display_name=Color and Opacity
       sections [{
         type=MovieSceneColorSection
         end_frame=2
         start_frame=0
         channels [
          Color.R {type=MovieSceneScriptingFloatChannel}
          Color.G {type=MovieSceneScriptingFloatChannel}
          Color.B {type=MovieSceneScriptingFloatChannel}
          Color.A {type=MovieSceneScriptingFloatChannel, keys [[0,0],[2,1]]}
         ]
        }]
      }]
    }
    FocusOutline {
     id=#3
     possessed_class=Image
     tracks [{
       type=MovieSceneFloatTrack
       display_name=Render Opacity
       sections [{type=MovieSceneFloatSection, end_frame=2, start_frame=0, channels [None {type=MovieSceneScriptingFloatChannel, keys [[0,0],[2,1]]}]}]
      }]
    }
    Button_Increase {
     id=#4
     possessed_class=IconTextButton_C
     tracks [{
       type=MovieSceneColorTrack
       display_name=Color and Opacity
       sections [{
         type=MovieSceneColorSection
         end_frame=2
         start_frame=0
         channels [
          Color.R {type=MovieSceneScriptingFloatChannel}
          Color.G {type=MovieSceneScriptingFloatChannel}
          Color.B {type=MovieSceneScriptingFloatChannel}
          Color.A {type=MovieSceneScriptingFloatChannel, keys [[0,0],[2,1]]}
         ]
        }]
      }]
    }
   ]}
 }]
graphs [
 Center On Widget {nodes [
   Center On Widget() {is=title K2Node_FunctionEntry}
   > CenterWidget!(execute, ToCenter:Widget Object Reference=Get TextRotator() -> (TextRotator:Text Rotator Object Reference)) -> (then)
  ]}
 EventGraph {nodes [
   "On Hovered (TextRotator)"!() -> (OutputDelegate, then, Button:Common Button Base Object Reference) {is=title K2Node_ComponentBoundEvent}
   > Call On Hovered!(execute:Exec, self:Rotator Selector Replay Settings Object Reference) -> (then:Exec) {is=title K2Node_CallDelegate}
   "On Button Clicked (Button_Decrease)"!() -> (OutputDelegate, then, Button:Common Button Legacy Object Reference) {is=title K2Node_ComponentBoundEvent}
   > ShiftTextLeft(self:Common Rotator Object Reference=Get TextRotator())
   "On Button Clicked (Button_Increase)"!() -> (OutputDelegate, then, Button:Common Button Legacy Object Reference) {is=title K2Node_ComponentBoundEvent}
   > ShiftTextRight(self:Common Rotator Object Reference=Get TextRotator())
   n1 = "On Rotated (TextRotator)"!() -> (OutputDelegate, then, Value:int=0) {is=title K2Node_ComponentBoundEvent}
   > n2 = Set SelectedIndex(SelectedIndex:int=@n1.Value) -> (Output_Get:int=0)
   n3 = Get SelectedIndex() -> (SelectedIndex:int=0)
   Call Selection Changed!(execute:Exec=[@n13.then, @n8.then, @n2.then], self:Rotator Selector Replay Settings Object Reference, Selected Index:int=@n3) -> (then:Exec) {is=title K2Node_CallDelegate}
   > SetScalarParameterValue(
    self=GetRotatorMaterial!() -> (ReturnValue:Material Instance Dynamic Object Reference)
    ParameterName=SelectedOption
    Value="To Float (Integer)"(InInt=@n3)
   )
   > SpawnSound2D!(
    execute
    Sound:Sound Base Object Reference=Get PressedSound() -> (PressedSound:Sound Base Object Reference)
    VolumeMultiplier:float=1
    PitchMultiplier:float=1
    StartTime:float=0
    ConcurrencySettings:Sound Concurrency Object Reference
    bPersistAcrossLevelTransition:bool=false
    bAutoDestroy:bool=true
   ) -> (then, ReturnValue:Audio Component Object Reference)
   ForceRotateRight() {is=title K2Node_CustomEvent}
   > n4 = Set SelectedIndex(SelectedIndex=int + int(A=1, B=Get SelectedIndex() -> (SelectedIndex=0))) -> (Output_Get=0)
   n5 = Length(TargetArray:Array of Texts=Get Options() -> (Options:Array of Texts))
   > n6 = Branch(Condition="integer > integer"(A=@n4, B=int - int(A=@n5)))
   > n7 = Set SelectedIndex(SelectedIndex=0) -> (Output_Get=0)
   n8 = SetSelectedItem!(execute=[@n7.then, @n6.else], self:Common Rotator Object Reference=Get TextRotator(), InValue:int=Get SelectedIndex() -> (SelectedIndex=0)) -> (then)
   ForceRotateLeft() {is=title K2Node_CustomEvent}
   > n9 = Set SelectedIndex(SelectedIndex=int - int(A=Get SelectedIndex() -> (SelectedIndex=0))) -> (Output_Get=0)
   > n10 = Branch(Condition="integer < integer"(A=@n9))
   n11 = Get Options()
   > n12 = Set SelectedIndex(SelectedIndex=int - int(A=Length(TargetArray=@n11))) -> (Output_Get=0)
   n13 = SetSelectedItem(execute=[@n12.then, @n10.else], self=Get TextRotator(), InValue=Get SelectedIndex() -> (SelectedIndex=0))
   Event PreConstruct()
   n14 = Get MyText!(self:Common Rotator Object Reference=Get TextRotator()) -> (MyText:Common Text Object Reference)
   > SetFont!(
    execute
    self=@n14
    InFontInfo_FontObject:Object Reference=&HeadingNow
    InFontInfo_FontMaterial:Object Reference
    InFontInfo_OutlineSettings:Font Outline Settings Structure
    InFontInfo_TypefaceFontName:Name=74 Regular
    InFontInfo_Size:float=18.5
    InFontInfo_LetterSpacing:int=0
    InFontInfo_SkewAmount:float=0
    InFontInfo_bForceMonospaced:bool=false
    InFontInfo_bMaterialIsStencil:bool=false
    InFontInfo_MonospacedWidth:float=1
   ) -> (then)
   > "SetText (Text)"(self=@n14, InText="LOCTEXT(\"{Options}\")")
   > SetColorAndOpacity!(
    execute
    self=@n14
    InColorAndOpacity_SpecifiedColor:Linear Color Structure=(R=0.7529,G=0.7529,B=0.7529,A=1)
    InColorAndOpacity_ColorUseRule:ESlateColorStylingMode Enum
   ) -> (then)
   n15 = Get WBP_UIKit_Backplate!(self:Setting Entry Background Object Reference=Get Background() -> (Background:Setting Entry Background Object Reference)) -> (WBP_UIKit_Backplate:WBP UIKit Backplate Object Reference)
   > SetBackplateMaterial!(
    execute
    self:WBP UIKit Backplate Base Object Reference=@n15
    Material Instance:Material Instance Object Reference=/FortniteCoreUI/UI/Replay/Materials/M_Replay_Backplate_Inst
   ) -> (then)
   > SetBrightness!(execute, self:WBP UIKit Backplate Base Object Reference=@n15, Brightness:E_UI_BackplateBrightness Enum=NewEnumerator1) -> (then)
   > Initialize!(
    execute
    self
    Display Text:Text=Get Title() -> (Title:Text)
    Options:Array of Texts=Get Options()
    Hover Text:Text
    Tab Tooltip:Common Text Object Reference
   ) -> (then)
   Event OnMouseLeave!() -> (OutputDelegate, then, MouseEvent:"Pointer Event Structure (by ref)") {is=title K2Node_Event}
   > "SetText (Text)"(self=Get Tab Tooltip Text() -> (Tab Tooltip Text:Common Text Object Reference))
   > PlayAnimationReverse(InAnimation=Get OnHover() -> (OnHover:Widget Animation Object Reference))
   Event OnMouseEnter()
   > Call On Focused!(execute:Exec, self:Rotator Selector Replay Settings Object Reference) -> (then:Exec) {is=title K2Node_CallDelegate}
   > "SetText (Text)"(self=Get Tab Tooltip Text(), InText=Get HoverText() -> (HoverText:Text))
   > PlayAnimationForward(InAnimation=Get OnHover())
   > SpawnSound2D(
    Sound=Get HoverSound() -> (HoverSound:Sound Base Object Reference)
    VolumeMultiplier=1
    PitchMultiplier=1
    StartTime=0
    bPersistAcrossLevelTransition=false
    bAutoDestroy=true
   )
   > ShowSettingDescription!(
    execute
    self:Fort Replay Context Object Reference=Get FortReplayContext!() -> (ReturnValue:Fort Replay Context Object Reference) {is=title K2Node_GetBlueprintContextEx}
    Header:"Text (by ref)"=Get Title()
    Description:"Text (by ref)"=Get Description() -> (Description:Text)
   ) -> (then)
  ]}
 FormatText {nodes [
   n1 = FormatText!() -> (then, TitleIn:Text) {is=title K2Node_FunctionEntry}
   > Set Loc_TextIn(Loc_TextIn:Text=@n1) -> (Output_Get:Text)
   > n2 = Switch on E_Replay_TextFormat!(execute, Selection:E_Replay_TextFormat Enum=Get Formatting() -> (Formatting:E_Replay_TextFormat Enum=NewEnumerator2)) -> (NewEnumerator2:Exec, NewEnumerator1:Exec, NewEnumerator3:Exec) {is=title K2Node_SwitchEnum}
   Return Node!(
    execute=@n2.NewEnumerator1
    TextOut:Text=Format Text!(
     Format="LOCTEXT(\"{indent once}{title}\")"
     indent once:Text=Get Loc_Indent() -> (Loc_Indent:Text)
     title:Text=Get Loc_TextIn() -> (Loc_TextIn:Text)
    ) -> (Result)
   )
   > Return Node(TextOut=Get Loc_TextIn())
   n3 = Get Loc_Indent()
   Return Node(
    execute=@n2.NewEnumerator3
    TextOut=Format Text!(Format="LOCTEXT(\"{indent once}{indent twice}{title}\")", indent once=@n3, indent twice:Text=@n3, title=Get Loc_TextIn()) -> (Result)
   )
  ]}
 GetRotatorMaterial {nodes [
   GetRotatorMaterial() {is=title K2Node_FunctionEntry}
   > n1 = GetDynamicMaterial(self:Border Object Reference=Get MainBorder!(self:Text Rotator Object Reference=Get TextRotator()) -> (MainBorder:Common Border Object Reference))
   > Return Node!(execute, ReturnValue:Material Instance Dynamic Object Reference=@n1)
  ]}
 Initialize {nodes [
   n1 = Initialize!() -> (then, Display Text:Text, Options:Array of Texts, Hover Text:Text, Tab Tooltip:Common Text Object Reference) {is=title K2Node_FunctionEntry}
   > Set Title(Title:Text=@n1) -> (Output_Get:Text)
   > "SetText (Text)"(
    self=Get Text_SettingName() -> (Text_SettingName:Common Text Object Reference)
    InText=FormatText!(TitleIn:Text=@n1) -> (TextOut:Text) {is=title K2Node_CallFunction}
   )
   > n2 = Set Options(Options:Array of Texts=@n1.Options) -> (Output_Get:Array of Texts)
   > PopulateTextLabels!(execute, self:Common Rotator Object Reference=Get TextRotator(), Labels:Array of Texts=@n2) -> (then)
   > SetScalarParameterValue(self=GetRotatorMaterial() {is=title K2Node_CallFunction}, ParameterName=Options, Value="To Float (Integer)"(InInt=Length(TargetArray=@n2)))
   > Set HoverText(HoverText:Text=@n1.Hover Text) -> (Output_Get:Text)
   > Set Tab Tooltip Text(Tab Tooltip Text:Common Text Object Reference=@n1.Tab Tooltip) -> (Output_Get:Common Text Object Reference)
   n3 = Get SelectedIndex() -> (SelectedIndex=0)
   > SetSelectedItem(self=Get TextRotator(), InValue=@n3)
   > SetScalarParameterValue(self=GetRotatorMaterial(), ParameterName=SelectedOption, Value="To Float (Integer)"(InInt=@n3))
   > Branch(Condition=Get bUseDefaultOption() -> (bUseDefaultOption:bool))
   > SetScalarParameterValue(self=GetRotatorMaterial(), ParameterName=DefaultOption, Value=Get DefaultOption() -> (DefaultOption:float))
  ]}
 On Hovered {nodes [On Hovered() {is=title K2Node_FunctionEntry}]}
 OnFocusReceived {nodes [
   OnFocusReceived!() -> (then, MyGeometry:Geometry Structure, InFocusEvent:Focus Event Structure) {is=title K2Node_FunctionEntry}
   > Center On Widget() {is=title K2Node_CallFunction}
   > Return Node(ReturnValue:Event Reply Structure=Unhandled())
  ]}
 OnFocused {nodes [OnFocused() {is=title K2Node_FunctionEntry}]}
 OnTextCommitted {nodes [OnTextCommitted!() -> (then, Text:double=~) {is=title K2Node_FunctionEntry}]}
 Selection Changed {nodes [Selection Changed!() -> (then, Selected Index:int) {is=title K2Node_FunctionEntry}]}
 SetHighlightedOption {nodes [
   n1 = SetHighlightedOption!() -> (then, bHighlight:bool=~, Index:int) {is=title K2Node_FunctionEntry}
   > n2 = Branch(Condition=@n1)
   n3 = "To Float (Integer)"(InInt=@n1.Index)
   > SetScalarParameterValue(self=GetRotatorMaterial(), ParameterName=DefaultOption, Value=@n3)
   SetScalarParameterValue(execute=@n2.else, self=GetRotatorMaterial(), ParameterName=SelectedOption, Value=@n3)
  ]}
 TextToFloat {nodes [
   n1 = Inputs!() -> (InText:"Text (by ref)")
   Outputs!(ReturnValue:double=String To Float(InString="To String (Text)"(InText=@n1)))
  ]}
 Update Options {nodes [
   n1 = Update Options!() -> (then, Options:Array of Texts) {is=title K2Node_FunctionEntry}
   > n2 = Set Options(Options=@n1)
   > PopulateTextLabels(self=Get TextRotator(), Labels=@n2)
   n3 = Length(TargetArray=@n2)
   > SetScalarParameterValue(self=GetRotatorMaterial(), ParameterName=Options, Value="To Float (Integer)"(InInt=@n3))
   > Return Node!(execute, bOptionsEmpty:bool="Equal (Integer)"(A=@n3))
  ]}
 Update Row Selector {nodes [
   n1 = Update Row Selector!() -> (then, Currently Selected:int) {is=title K2Node_FunctionEntry}
   > n2 = Set SelectedIndex(SelectedIndex=@n1) -> (Output_Get=0)
   > SetSelectedItem(self=Get TextRotator(), InValue=@n2)
   > SetScalarParameterValue(self=GetRotatorMaterial(), ParameterName=SelectedOption, Value="To Float (Integer)"(InInt=@n2))
   > Branch(Condition=Get bUseDefaultOption())
   > SetScalarParameterValue(self=GetRotatorMaterial(), ParameterName=DefaultOption, Value=Get DefaultOption())
  ]}
 UpdateOptionsText {nodes [
   n1 = UpdateOptionsText!() -> (then, NewText:Text) {is=title K2Node_FunctionEntry}
   > "SetText (Text)"(self=Get MyText(self=Get TextRotator()), InText=@n1)
  ]}
]
variables [
 bUseDefaultOption {type=boolean}
 DefaultOption {type=number}
 Description {type=string, default=""}
 Formatting {type=E_Replay_TextFormat, default=None}
 HoverSound {type=SoundBase, default=/Game/Sounds/Fort_UI_Sounds/FrontEnd/DefaultSounds/PressedButtons/MSS_UI_Button_Hover_Default}
 HoverText {type=string, default=""}
 On Hovered {}
 OnFocused {}
 OnTextCommitted {}
 Options {type="array<string>", default []}
 PressedSound {type=SoundBase, default=/Game/Sounds/Fort_UI_Sounds/FrontEnd/DefaultSounds/PressedButtons/MSS_UI_Button_Pressed_Default}
 SelectedIndex {type=integer}
 Selection Changed {}
 Tab Tooltip Text {type=CommonTextBlock, default=null}
 Title {type=string, default="{Setting Name}"}
]
widgets {
 info {inheritedWidgetCount=0, namedSlotCount=1, parentClass=&CommonUserWidget, rootWidgetClass=&Overlay, widgetCount=13}
 widgets [
  Overlay_0 {widgetClassPath=&Overlay, widget=Overlay_0}
  MasterSizeBox {
   widgetClassPath=&SizeBox
   bIsVariable=true
   parent=Overlay_0
   widget=MasterSizeBox
   properties {bOverride_WidthOverride=true, widthOverride=650}
   slot {class=OverlaySlot, properties {horizontalAlignment=HAlign_Fill, verticalAlignment=VAlign_Fill}}
  }
  Background {
   widgetClassPath=/FortniteCoreUI/Settings/SettingEntries/SettingEntryBackground
   bIsVariable=true
   parent=MasterSizeBox
   widget=Background
   slot {class=SizeBoxSlot}
  }
  Overlay {widgetClassPath=&Overlay, namedSlotHost=Background, widget=Overlay}
  FortResponsivePanel_0 {
   widgetClassPath=/Script/FortniteUI.FortResponsivePanel
   parent=Overlay
   widget=FortResponsivePanel_0
   properties {bCanStackVertically=false, bOverride_Cursor=true}
   slot {class=OverlaySlot, properties {horizontalAlignment=HAlign_Fill, verticalAlignment=VAlign_Fill}}
  }
  SettingNameSizeBox {
   widgetClassPath=&SizeBox
   bIsVariable=true
   parent=FortResponsivePanel_0
   widget=SettingNameSizeBox
   properties {bOverride_MinDesiredHeight=true, minDesiredHeight=64, minDesiredWidth=300}
   slot {class=FortResponsivePanelSlot}
  }
  Text_SettingName {
   widgetClassPath=&CommonTextBlock
   bIsVariable=true
   parent=SettingNameSizeBox
   widget=Text_SettingName
   properties {
    scrollStyle=&TextScrollStyle-Base_C
    text="{Setting Name}"
    visibility=Visible
    colorAndOpacity {colorUseRule=UseColor_Specified, specifiedColor=<rgba 0.521 0.521 0.521 0.8>}
    font {
     bForceMonospaced=false
     bMaterialIsStencil=false
     fontMaterial=null
     fontObject=&HeadingNow
     letterSpacing=0
     monospacedWidth=1
     size=18
     skewAmount=0
     typefaceFontName=75 Medium
     outlineSettings {
      bApplyOutlineToDropShadows=false
      bMiteredCorners=false
      bSeparateFillAlpha=false
      outlineMaterial=null
      outlineSize=0
      outlineColor=<rgba 0 0 0 1>
     }
    }
    shadowOffset=<0 0>
   }
   slot {class=SizeBoxSlot, properties {horizontalAlignment=HAlign_Left, verticalAlignment=VAlign_Center, padding=<ltrb 24 0 10 0>}}
  }
  OptionsSizeBox {
   widgetClassPath=&SizeBox
   bIsVariable=true
   parent=FortResponsivePanel_0
   widget=OptionsSizeBox
   properties {minDesiredWidth=896.5338}
   slot {class=FortResponsivePanelSlot}
  }
  PanelValue {widgetClassPath=&HorizontalBox, parent=OptionsSizeBox, widget=PanelValue, slot {class=SizeBoxSlot}}
  Button_Decrease {
   widgetClassPath=&IconTextButton
   bIsVariable=true
   parent=PanelValue
   widget=Button_Decrease
   properties {
    bIsFocusable=false
    buttonText=""
    onButtonBaseClicked=(&self:WidgetTree.Button_Decrease.HandleOnButtonClicked)
    onButtonBaseDoubleClicked=(&self:WidgetTree.Button_Decrease.HandleOnButtonDoubleClicked)
    onButtonBaseHovered=(&self:WidgetTree.Button_Decrease.HandleOnButtonHovered)
    onButtonBaseUnhovered=(&self:WidgetTree.Button_Decrease.HandleOnButtonUnhovered)
    onSelectedChangedBase=(&self:WidgetTree.Button_Decrease.HandleOnSelectedChanged)
    style=/FortUILibrary/Foundation/Buttons/ButtonStyles/ButtonStyleClearDismissTooltip.ButtonStyleClearDismissTooltip_C
    touchMethod=PreciseTap
    colorAndOpacity=<rgba 1 1 1 0>
    hoveredSlateSoundOverride {resourceObject=null}
    iconBrush {
     bIsDynamicallyLoaded=false
     drawAs=Image
     imageType=FullColor
     mirroring=NoMirror
     resourceName=None
     resourceObject=/FortUILibrary/Textures/Icons/ToBeDeprecated/T_UI_Icon_Global_Chevron_Left
     tiling=NoTile
     imageSize=<32 32>
     margin=<ltrb 0 0 0 0>
     outlineSettings {
      bUseBrushTransparency=false
      roundingType=HalfHeightRadius
      width=0
      color {colorUseRule=UseColor_Specified, specifiedColor=<rgba 0 0 0 0>}
      cornerRadii=<0 0 0 1>
     }
     tintColor {colorUseRule=UseColor_Specified, specifiedColor=<rgba 1 1 1 1>}
     uVRegion {bIsValid=false, max=<0 0>, min=<0 0>}
    }
    pressedSlateSoundOverride {resourceObject=null}
   }
   slot {class=HorizontalBoxSlot, properties {verticalAlignment=VAlign_Center}}
  }
  TextRotator {
   widgetClassPath=/FortUILibrary/Foundation/Buttons/TextRotator
   bIsVariable=true
   parent=PanelValue
   widget=TextRotator
   properties {
    bChangeAppearance=true
    image=/FortniteCoreUI/Settings/Materials/MI_SettingsRotator
    style=null
    foregroundColor {colorUseRule=UseColor_Specified, specifiedColor=<rgba 1 0 1 1>}
   }
   slot {class=HorizontalBoxSlot, properties {size {sizeRule=Fill, value=1}}}
  }
  Button_Increase {
   widgetClassPath=&IconTextButton
   bIsVariable=true
   parent=PanelValue
   widget=Button_Increase
   properties {
    bIsFocusable=false
    buttonText=""
    onButtonBaseClicked=(&self:WidgetTree.Button_Increase.HandleOnButtonClicked)
    onButtonBaseDoubleClicked=(&self:WidgetTree.Button_Increase.HandleOnButtonDoubleClicked)
    onButtonBaseHovered=(&self:WidgetTree.Button_Increase.HandleOnButtonHovered)
    onButtonBaseUnhovered=(&self:WidgetTree.Button_Increase.HandleOnButtonUnhovered)
    onSelectedChangedBase=(&self:WidgetTree.Button_Increase.HandleOnSelectedChanged)
    style=&ButtonStyleClearDismissTooltip_C
    touchMethod=PreciseTap
    colorAndOpacity=<rgba 1 1 1 0>
    hoveredSlateSoundOverride {resourceObject=null}
    iconBrush {
     bIsDynamicallyLoaded=false
     drawAs=Image
     imageType=FullColor
     mirroring=NoMirror
     resourceName=None
     resourceObject=/FortUILibrary/Textures/Icons/ToBeDeprecated/T_UI_Icon_Global_Chevron_Right
     tiling=NoTile
     imageSize=<32 32>
     margin=<ltrb 0 0 0 0>
     outlineSettings {
      bUseBrushTransparency=false
      roundingType=HalfHeightRadius
      width=0
      color {colorUseRule=UseColor_Specified, specifiedColor=<rgba 0 0 0 0>}
      cornerRadii=<0 0 0 1>
     }
     tintColor {colorUseRule=UseColor_Specified, specifiedColor=<rgba 1 1 1 1>}
     uVRegion {bIsValid=false, max=<0 0>, min=<0 0>}
    }
    pressedSlateSoundOverride {resourceObject=null}
   }
   slot {class=HorizontalBoxSlot}
  }
  FocusOutline {
   widgetClassPath=&Image
   bIsVariable=true
   parent=Overlay
   widget=FocusOutline
   properties {
    pixelSnapping=Disabled
    renderOpacity=0
    visibility=SelfHitTestInvisible
    brush {
     bIsDynamicallyLoaded=false
     drawAs=Image
     imageType=NoImage
     mirroring=NoMirror
     resourceName=None
     resourceObject=/FortniteCoreUI/Settings/Art/MI_UIKit_Settings_Outline
     tiling=NoTile
     imageSize=<32 32>
     margin=<ltrb 0 0 0 0>
     outlineSettings {
      bUseBrushTransparency=false
      roundingType=HalfHeightRadius
      width=0
      color {colorUseRule=UseColor_Specified, specifiedColor=<rgba 0 0 0 0>}
      cornerRadii=<0 0 0 0>
     }
     tintColor {colorUseRule=UseColor_Specified, specifiedColor=<rgba 1 1 1 1>}
     uVRegion {bIsValid=false, max=<0 0>, min=<0 0>}
    }
   }
   slot {class=OverlaySlot, properties {horizontalAlignment=HAlign_Fill, verticalAlignment=VAlign_Fill, padding=<ltrb -16 -16 -16 -16>}}
  }
 ]
}
