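# Replay UI tab button, derived from the UIKit selectable button, that collapses and expands its label area.
# It starts collapsed and follows the Fort Replay Context's collapse event through the collapsible interface.
# The OnCollapsed animation fades and shifts the content while driving a value that linearly eases width to zero, reversing faster to expand.
# It also styles the background corners and focus outline, and fills preview text and icon from a viewmodel.
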WidgetBlueprint /FortniteCoreUI/UI/Replay/WBP_Replay_TabButton
animations [OnCollapsed {
  binding_count=2
  length_seconds=0.35
  object=&self:OnCollapsed
  sequence {bindings [
    Main_Stack {
     id=#1
     possessed_class=StackBox
     tracks [
      {
       type=MovieSceneFloatTrack
       display_name=Render Opacity
       sections [{
         type=MovieSceneFloatSection
         end_frame=7
         start_frame=0
         channels [None {type=MovieSceneScriptingFloatChannel, keys [[0,1],[2.8463,0],[7,0]]}]
        }]
      }
      {
       type=MovieScene2DTransformTrack
       sections [{
         type=MovieScene2DTransformSection
         end_frame=7
         start_frame=0
         channels [Translation.X {type=MovieSceneScriptingFloatChannel, keys [[0,0],[2.8463,-10],[7,-10]]}]
        }]
      }
     ]
    }
    WBP_Replay_TabButton {
     id=#2
     possessed_class=WBP_Replay_TabButton_C
     tracks [{
       type=MovieSceneDoubleTrack
       display_name=On Collapse Anim Value
       sections [{type=MovieSceneDoubleSection, end_frame=7, start_frame=0, channels [None {type=MovieSceneScriptingDoubleChannel, keys [[2.4,0],[7,1]]}]}]
      }]
    }
   ]}
 }]
graphs [
 EventGraph {nodes [
   n1 = HandleCollapseTab!() -> (OutputDelegate, then, IsCollapsed:bool=~) {is=title K2Node_CustomEvent}
   > SetIsCollapsed!(execute, self:BPI UIKit Collapsible Interface, IsCollapsed:bool=@n1.IsCollapsed, SkipAnimation:bool=false) -> (then)
   n2 = Event PreConstruct()
   > Parent: PreConstruct(IsDesignTime=@n2.IsDesignTime)
   > SetWidthOverride(self=Get Main_StackSize() -> (Main_StackSize:Size Box Object Reference), InWidthOverride=0)
   > Bind Event to On Collapse Tab(
    self:Fort Replay Context Object Reference=Get FortReplayContext!() -> (ReturnValue:Fort Replay Context Object Reference) {is=title K2Node_GetBlueprintContextEx}
    Delegate=@n1
   ) {is=title K2Node_AddDelegate}
   Event Construct()
   > Parent: Construct()
   > n3 = Cast To WBP_UIKit_BlockBackground_Tab_Enhanced!(execute, Object=Get Block_Background() -> (Block_Background:WBP UIKit Block Background Base Object Reference)) -> (then, CastFailed, AsWBP UIKit Block Background Tab Enhanced:WBP UIKit Block Background Tab Enhanced Object Reference) {is=title K2Node_DynamicCast}
   > SetCornerRadiuses!(execute, self:WBP UIKit Block Background Tab Enhanced Object Reference=@n3, CornerRadius:float=12, RadiusInPixels:bool=true) -> (then)
   > SetPadding!(
    execute
    self:Grid Slot Object Reference=SlotAsGridSlot(Widget=Get FocusOutline!(self:WBP UIKit Block Background Tab Enhanced Object Reference=@n3) -> (FocusOutline:Image Object Reference))
    InPadding_Left:float=-15
    InPadding_Top:float=-15
    InPadding_Right:float=-15
    InPadding_Bottom:float=-15
   ) -> (then)
   n4 = Event SetIsCollapsed!() -> (OutputDelegate, then, IsCollapsed:bool, SkipAnimation:bool) {is=title K2Node_Event}
   > n5 = Branch(Condition=@n4.IsCollapsed)
   > n6 = Branch(Condition=Get bCollapsed() -> (bCollapsed:bool))
   PlayAnimationForward(execute=@n6.else, InAnimation=Get OnCollapsed() -> (OnCollapsed:Widget Animation Object Reference))
   > Set bCollapsed(bCollapsed:bool=true) -> (Output_Get:bool)
   > SetOnCollapseAnimValue!(execute, self, Value:double=1) -> (then)
   Branch(execute=@n5.else, Condition=Get bCollapsed())
   > PlayAnimationReverse(InAnimation=Get OnCollapsed(), PlaybackSpeed=1.75)
   > Set bCollapsed(bCollapsed=false)
  ]}
 GetRemainingAnimTime {nodes [
   GetRemainingAnimTime() {is=title K2Node_FunctionEntry}
   > Return Node!(execute, FunctionIsImplemented:bool=false, RemainingTime:double=0, TotalTime:double=0)
  ]}
 SetOnCollapseAnimValue {nodes [
   n1 = SetOnCollapseAnimValue!() -> (then, Value:double=~) {is=title K2Node_FunctionEntry}
   > SetWidthOverride(
    self=Get Main_StackSize()
    InWidthOverride=Ease!(Function:EEasingFunc Enum=Linear, Alpha:double=@n1, A:float=Get MaxTextWidth() -> (MaxTextWidth:float), B:float) -> (Result:float=~) {is=title K2Node_EaseFunction}
   )
  ]}
 SetPreviewData {nodes [
   n1 = SetPreviewData!() -> (then, VM:MVVM Base Viewmodel Object Reference) {is=title K2Node_FunctionEntry}
   > n2 = Cast To UIKitSelectableButtonViewModel!(execute, Object=@n1) -> (then, CastFailed, AsUI Kit Selectable Button View Model:UI Kit Selectable Button View Model Object Reference) {is=title K2Node_DynamicCast}
   > SetText!(
    execute
    self:WBP UIKit Button Generic Base Object Reference
    Text:Text=Get Label!(self:UI Kit Selectable Button View Model Object Reference=@n2) -> (Label:Text)
   ) -> (then)
   > SetIcon!(
    execute
    self:WBP UIKit Button Generic Base Object Reference
    Icon:Object Reference=Get Icon!(self:UI Kit Selectable Button View Model Object Reference=@n2) -> (Icon:Texture 2D Object Reference)
   ) -> (then)
  ]}
]
variables [
 bCollapsed {type=boolean, default=true}
 MaxTextWidth {type=number, default=222}
 OnCollapseAnimValue {type=number}
]
widgets {
 info {
  inheritedWidgetCount=0
  parentClass=/FortUILibrary/UIKit/Buttons/WBP_UIKit_Button_Selectable.WBP_UIKit_Button_Selectable_C
  rootWidgetClass=null
  widgetCount=0
 }
 widgets []
}
